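# Physics asset for the skirt flaps on the QuirkLuster female medium player cosmetic body.
# Left and right flap chains hang from the pelvis, with tiny base capsules and limited-angle joints driving each dynamic segment's range.
# Pelvis, thigh, calf and heel capsules give the flaps body shapes to collide with, linked by free-rotation constraints.
# A huge root sphere, sunk below the origin, approximates a ground plane.
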PhysicsAsset /BRCosmetics/Characters/Player/Female/Medium/Bodies/F_MED_QuirkLuster/Meshes/F_MED_QuirkLuster_Physics_Skirt
bodies {
 base_SkirtFlap_00_l [{shapeType=Capsule, length=0.505, radius=0.505, shapeName=base_SkirtFlap_00_l_capsule, rotation=<rot 0 -90 0>}]
 base_SkirtFlap_00_r [{shapeType=Capsule, length=0.505, radius=0.505, shapeName=base_SkirtFlap_00_r_capsule, rotation=<rot 0 -90 0>}]
 base_SkirtFlap_01_l [{shapeType=Capsule, length=0.505, radius=0.505, shapeName=base_SkirtFlap_01_l_capsule, rotation=<rot 0 -90 0>}]
 base_SkirtFlap_01_r [{shapeType=Capsule, length=0.505, radius=0.505, shapeName=base_SkirtFlap_01_r_capsule, rotation=<rot 0 -90 0>}]
 calf_l [{shapeType=Capsule, length=39.4027, radius=7.8236, shapeName=calf_l_capsule, center=<-21.7022 -1.3542 -0.1744>, rotation=<rot 90 6.2534 0>}]
 calf_r [{shapeType=Capsule, length=39.4027, radius=7.8236, shapeName=calf_r_capsule, center=<21.7022 1.3542 0.1744>, rotation=<rot 90 -173.7466 0>}]
 dyn_SkirtFlap_00_l [{
   shapeType=Capsule
   length=11.6967
   radius=1.2112
   shapeName=dyn_SkirtFlap_00_l_capsule
   center=<5.9869 -0.8441 0.6074>
   rotation=<rot -0.4259 -0.8233 -74.7099>
  }]
 dyn_SkirtFlap_00_r [{
   shapeType=Capsule
   length=11.6967
   radius=1.2112
   shapeName=dyn_SkirtFlap_00_r_capsule
   center=<5.9849 -0.896 0.4385>
   rotation=<rot -0.4257 0.8233 -108.0002>
  }]
 dyn_SkirtFlap_01_l [
  {
   shapeType=Capsule
   length=9.9559
   radius=2.6789
   shapeName=dyn_SkirtFlap_01_l_capsule
   center=<10.8832 -0.1773 0.4068>
   rotation=<rot -0.1699 1.2912 -82.8>
  }
  {
   shapeType=Capsule
   length=10.3681
   radius=2.2668
   shapeName=dyn_SkirtFlap_01_l_capsule
   center=<5.6392 -0.0636 0.2839>
   rotation=<rot -0.1699 1.2912 -82.8>
  }
 ]
 dyn_SkirtFlap_01_r [
  {
   shapeType=Capsule
   length=9.9559
   radius=2.6789
   shapeName=dyn_SkirtFlap_01_r_capsule
   center=<10.8832 -0.1773 0.4312>
   rotation=<rot -0.1699 -1.2912 -97.2>
  }
  {
   shapeType=Capsule
   length=10.3681
   radius=2.2668
   shapeName=dyn_SkirtFlap_01_r_capsule
   center=<5.6392 -0.0636 0.2839>
   rotation=<rot -0.1699 -1.2912 -97.2>
  }
 ]
 dyn_SkirtFlap_02_l [
  {
   shapeType=Capsule
   length=9.4219
   radius=2.4478
   shapeName=dyn_SkirtFlap_01_l_capsule
   center=<12.5339 1.393 0.2082>
   rotation=<rot -0.17 1.2912 -78.8391>
  }
  {
   shapeType=Capsule
   length=10.3681
   radius=2.2668
   shapeName=dyn_SkirtFlap_01_l_capsule
   center=<5.6161 0.9486 0.156>
   rotation=<rot -0.17 1.2912 -78.8391>
  }
 ]
 dyn_SkirtFlap_02_r [
  {
   shapeType=Capsule
   length=9.4219
   radius=2.4478
   shapeName=dyn_SkirtFlap_01_r_capsule
   center=<12.5339 -1.393 0.2082>
   rotation=<rot -0.17 -1.2912 -101.1609>
  }
  {
   shapeType=Capsule
   length=10.3681
   radius=2.2668
   shapeName=dyn_SkirtFlap_01_r_capsule
   center=<5.6161 -0.9486 -0.156>
   rotation=<rot -0.17 -1.2912 -101.1609>
  }
 ]
 dyn_SkirtFlap_03_l [
  {
   shapeType=Capsule
   length=8.5007
   radius=2.6537
   shapeName=dyn_SkirtFlap_01_l_capsule
   center=<9.4885 2.5134 -0.0593>
   rotation=<rot -0.1701 1.2912 -75.6117>
  }
  {
   shapeType=Capsule
   length=9.0493
   radius=2.2668
   shapeName=dyn_SkirtFlap_01_l_capsule
   center=<3.6152 2.3604 -0.1415>
   rotation=<rot -0.1701 1.2912 -75.6117>
  }
 ]
 dyn_SkirtFlap_03_r [
  {
   shapeType=Capsule
   length=8.285
   radius=2.6537
   shapeName=dyn_SkirtFlap_01_r_capsule
   center=<9.497 -2.1493 -0.2115>
   rotation=<rot -0.1701 -1.2912 -104.3883>
  }
  {
   shapeType=Capsule
   length=9.323
   radius=2.2668
   shapeName=dyn_SkirtFlap_01_r_capsule
   center=<3.6167 -2.2972 0.0191>
   rotation=<rot -0.1701 -1.2912 -104.3883>
  }
 ]
 dyn_SkirtFlap_04_l [
  {
   shapeType=Capsule
   length=4.925
   radius=2.6978
   shapeName=dyn_SkirtFlap_01_l_capsule
   center=<7.2386 -0.6647 0.0883>
   rotation=<rot -9.5889 41.8077 -82.0942>
  }
  {
   shapeType=Capsule
   length=8.0448
   radius=2.5168
   shapeName=dyn_SkirtFlap_01_l_capsule
   center=<3.6464 1.0221 0.2019>
   rotation=<rot -0.1701 1.2912 -75.6117>
  }
  {
   shapeType=Capsule
   length=4.925
   radius=2.6978
   shapeName=dyn_SkirtFlap_01_l_capsule
   center=<9.6842 -3.4823 0.3817>
   rotation=<rot -14.3782 88.3342 -89.4312>
  }
 ]
 dyn_SkirtFlap_04_r [
  {
   shapeType=Capsule
   length=4.925
   radius=2.6978
   shapeName=dyn_SkirtFlap_01_r_capsule
   center=<7.2386 0.6647 -0.0883>
   rotation=<rot -9.5889 -41.8077 -97.9058>
  }
  {
   shapeType=Capsule
   length=8.0448
   radius=2.5168
   shapeName=dyn_SkirtFlap_01_r_capsule
   center=<3.6464 -1.0221 -0.2019>
   rotation=<rot -0.1701 -1.2912 -104.3883>
  }
  {
   shapeType=Capsule
   length=4.925
   radius=2.6978
   shapeName=dyn_SkirtFlap_01_r_capsule
   center=<9.6842 3.4823 -0.3817>
   rotation=<rot -14.3782 -88.3342 -90.5688>
  }
 ]
 heel_foot_l [{
   shapeType=Capsule
   length=6.3021
   radius=9.5345
   shapeName=heel_foot_l_capsule
   center=<-6.4216 3.7294 -0.301>
   rotation=<rot 0 10.1371 -89.6363>
  }]
 heel_foot_r [{
   shapeType=Capsule
   length=6.3021
   radius=9.5345
   shapeName=heel_foot_r_capsule
   center=<6.4216 -3.7294 0.301>
   rotation=<rot 0 10.1371 90.3637>
  }]
 pelvis [{
   shapeType=Capsule
   length=15.0014
   radius=10.2802
   shapeName=pelvis_capsule
   center=<-1.7181 1.0058 -0.0002>
   rotation=<rot -0.0067 -90 -0.0034>
  }]
 root [{shapeType=Capsule, length=0, radius=400, shapeName=root_capsule, center=<0 0 -396>, rotation=<rot 0 -90 0>}]
 thigh_l [{
   shapeType=Capsule
   length=18.7947
   radius=7.9753
   shapeName=thigh_twist_01_l_capsule
   center=<-29.5537 0.5437 -0.5065>
   rotation=<rot 0 -90 -90>
  }]
 thigh_r [{
   shapeType=Capsule
   length=18.7947
   radius=7.9753
   shapeName=thigh_twist_01_r_capsule
   center=<29.5537 -0.5437 0.5065>
   rotation=<rot 0 -90 90>
  }]
 thigh_twist_01_l [{
   shapeType=Capsule
   length=13.8594
   radius=8.7802
   shapeName=thigh_twist_01_l_capsule
   center=<17.9162 0.5437 -0.5065>
   rotation=<rot 0 -90 -90>
  }]
 thigh_twist_01_r [{
   shapeType=Capsule
   length=13.6699
   radius=8.9698
   shapeName=thigh_twist_01_r_capsule
   center=<-17.9162 -0.5437 0.5065>
   rotation=<rot 0 -90 90>
  }]
}
constraints [
 {
  bone1Name=thigh_twist_01_l
  bone2Name=thigh_l
  swing1LimitDegrees=45
  swing1Motion=Free
  swing2LimitDegrees=45
  swing2Motion=Free
  twistLimitDegrees=45
  twistMotion=Free
 }
 {
  bone1Name=calf_l
  bone2Name=thigh_l
  swing1LimitDegrees=45
  swing1Motion=Free
  swing2LimitDegrees=45
  swing2Motion=Free
  twistLimitDegrees=45
  twistMotion=Free
 }
 {
  bone1Name=heel_foot_l
  bone2Name=calf_l
  swing1LimitDegrees=45
  swing1Motion=Free
  swing2LimitDegrees=45
  swing2Motion=Free
  twistLimitDegrees=45
  twistMotion=Free
 }
 {
  bone1Name=thigh_l
  bone2Name=pelvis
  swing1LimitDegrees=45
  swing1Motion=Free
  swing2LimitDegrees=45
  swing2Motion=Free
  twistLimitDegrees=45
  twistMotion=Free
 }
 {
  bone1Name=base_SkirtFlap_00_l
  bone2Name=pelvis
  swing1LimitDegrees=45
  swing1Motion=Free
  swing2LimitDegrees=45
  swing2Motion=Free
  twistLimitDegrees=45
  twistMotion=Free
 }
 {bone1Name=dyn_SkirtFlap_00_l, bone2Name=base_SkirtFlap_00_l, swing1LimitDegrees=10, swing2LimitDegrees=45, twistLimitDegrees=5}
 {
  bone1Name=base_SkirtFlap_01_l
  bone2Name=dyn_SkirtFlap_00_l
  swing1LimitDegrees=45
  swing1Motion=Free
  swing2LimitDegrees=45
  swing2Motion=Free
  twistLimitDegrees=45
  twistMotion=Free
 }
 {bone1Name=dyn_SkirtFlap_01_l, bone2Name=base_SkirtFlap_01_l, swing1LimitDegrees=10, swing2LimitDegrees=45, twistLimitDegrees=5}
 {bone1Name=dyn_SkirtFlap_02_l, bone2Name=dyn_SkirtFlap_01_l, swing1LimitDegrees=10, swing2LimitDegrees=45, twistLimitDegrees=5}
 {bone1Name=dyn_SkirtFlap_03_l, bone2Name=dyn_SkirtFlap_02_l, swing1LimitDegrees=10, swing2LimitDegrees=45, twistLimitDegrees=5}
 {bone1Name=dyn_SkirtFlap_04_l, bone2Name=dyn_SkirtFlap_03_l, swing1LimitDegrees=10, swing2LimitDegrees=45, twistLimitDegrees=5}
 {
  bone1Name=base_SkirtFlap_00_r
  bone2Name=pelvis
  swing1LimitDegrees=45
  swing1Motion=Free
  swing2LimitDegrees=45
  swing2Motion=Free
  twistLimitDegrees=45
  twistMotion=Free
 }
 {bone1Name=dyn_SkirtFlap_00_r, bone2Name=base_SkirtFlap_00_r, swing1LimitDegrees=10, swing2LimitDegrees=45, twistLimitDegrees=5}
 {
  bone1Name=base_SkirtFlap_01_r
  bone2Name=dyn_SkirtFlap_00_r
  swing1LimitDegrees=45
  swing1Motion=Free
  swing2LimitDegrees=45
  swing2Motion=Free
  twistLimitDegrees=45
  twistMotion=Free
 }
 {bone1Name=dyn_SkirtFlap_01_r, bone2Name=base_SkirtFlap_01_r, swing1LimitDegrees=10, swing2LimitDegrees=45, twistLimitDegrees=5}
 {bone1Name=dyn_SkirtFlap_02_r, bone2Name=dyn_SkirtFlap_01_r, swing1LimitDegrees=10, swing2LimitDegrees=45, twistLimitDegrees=5}
 {bone1Name=dyn_SkirtFlap_03_r, bone2Name=dyn_SkirtFlap_02_r, swing1LimitDegrees=10, swing2LimitDegrees=45, twistLimitDegrees=5}
 {bone1Name=dyn_SkirtFlap_04_r, bone2Name=dyn_SkirtFlap_03_r, swing1LimitDegrees=10, swing2LimitDegrees=45, twistLimitDegrees=5}
 {
  bone1Name=thigh_r
  bone2Name=pelvis
  swing1LimitDegrees=45
  swing1Motion=Free
  swing2LimitDegrees=45
  swing2Motion=Free
  twistLimitDegrees=45
  twistMotion=Free
 }
 {
  bone1Name=calf_r
  bone2Name=thigh_r
  swing1LimitDegrees=45
  swing1Motion=Free
  swing2LimitDegrees=45
  swing2Motion=Free
  twistLimitDegrees=45
  twistMotion=Free
 }
 {
  bone1Name=heel_foot_r
  bone2Name=calf_r
  swing1LimitDegrees=45
  swing1Motion=Free
  swing2LimitDegrees=45
  swing2Motion=Free
  twistLimitDegrees=45
  twistMotion=Free
 }
 {
  bone1Name=thigh_twist_01_r
  bone2Name=thigh_r
  swing1LimitDegrees=45
  swing1Motion=Free
  swing2LimitDegrees=45
  swing2Motion=Free
  twistLimitDegrees=45
  twistMotion=Free
 }
 {
  bone1Name=pelvis
  bone2Name=root
  swing1LimitDegrees=45
  swing1Motion=Free
  swing2LimitDegrees=45
  swing2Motion=Free
  twistLimitDegrees=45
  twistMotion=Free
 }
]
