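# Physics asset for the left and right tassels on the F_MED_QuirkLuster female player body, giving them limited secondary swing.
# On each side, a main tassel capsule hangs from a base body with tight limits, and eight tassel capsules attach to it in a fan with looser limits.
# Unconstrained upper-arm twist capsules appear to serve as arm colliders.
# Limits are not fully mirrored: the main joints swap swing axes and left tassel 01 swings less.
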PhysicsAsset /BRCosmetics/Characters/Player/Female/Medium/Bodies/F_MED_QuirkLuster/Meshes/F_MED_QuirkLuster_Physics_Tassles
bodies {
 base_tassels_main_l [{shapeType=Capsule, length=0, radius=0.5088, shapeName=base_tassels_main_l_capsule, center=<-2.0488 -0.3271 0>, rotation=<rot 0 -90 0>}]
 base_tassels_main_r [{shapeType=Capsule, length=0, radius=0.5088, shapeName=base_tassels_main_r_capsule, center=<2.0488 0.3271 0>, rotation=<rot 0 -90 180>}]
 dyn_tassels_01_l [{
   shapeType=Capsule
   length=1.0685
   radius=0.8616
   shapeName=dyn_tassels_01_l_capsule
   center=<-2.9808 -0.1812 0.0043>
   rotation=<rot 8.1372 -90 -8.5318>
  }]
 dyn_tassels_01_r [{
   shapeType=Capsule
   length=1.0685
   radius=0.8616
   shapeName=dyn_tassels_01_r_capsule
   center=<2.9808 0.1812 -0.0043>
   rotation=<rot 8.1372 -90 171.4682>
  }]
 dyn_tassels_02_l [{
   shapeType=Capsule
   length=1.9007
   radius=0.8616
   shapeName=dyn_tassels_01_l_capsule
   center=<-2.8 -0.2173 0.0756>
   rotation=<rot 4.7 -89.7491 -4.1345>
  }]
 dyn_tassels_02_r [{
   shapeType=Capsule
   length=1.9007
   radius=0.8616
   shapeName=dyn_tassels_01_r_capsule
   center=<2.8 0.2173 -0.0756>
   rotation=<rot 4.7 -89.7491 175.8655>
  }]
 dyn_tassels_03_l [{
   shapeType=Capsule
   length=1.8339
   radius=0.9098
   shapeName=dyn_tassels_03_l_capsule
   center=<-2.7551 -0.0537 0.1516>
   rotation=<rot 9.3446 -90.5034 7.0273>
  }]
 dyn_tassels_03_r [{
   shapeType=Capsule
   length=1.8339
   radius=0.9098
   shapeName=dyn_tassels_03_r_capsule
   center=<2.7551 0.0537 -0.1516>
   rotation=<rot 9.3446 -90.5034 -172.9727>
  }]
 dyn_tassels_04_l [{
   shapeType=Capsule
   length=1.5471
   radius=0.9347
   shapeName=dyn_tassels_04_l_capsule
   center=<-2.9754 0.0955 0.0331>
   rotation=<rot -7.4602 -90 -3.9619>
  }]
 dyn_tassels_04_r [{
   shapeType=Capsule
   length=1.5471
   radius=0.9347
   shapeName=dyn_tassels_04_r_capsule
   center=<2.9754 -0.0955 -0.0331>
   rotation=<rot -7.4602 -90 176.0381>
  }]
 dyn_tassels_05_l [{
   shapeType=Capsule
   length=1.1575
   radius=0.8161
   shapeName=dyn_tassels_05_l_capsule
   center=<-3.1147 0.0437 -0.0406>
   rotation=<rot -3.3952 -90 4.7708>
  }]
 dyn_tassels_05_r [{
   shapeType=Capsule
   length=1.1575
   radius=0.8161
   shapeName=dyn_tassels_05_r_capsule
   center=<3.1147 -0.0437 0.0406>
   rotation=<rot -3.3952 -90 -175.2292>
  }]
 dyn_tassels_06_l [{
   shapeType=Capsule
   length=1.3848
   radius=0.9084
   shapeName=dyn_tassels_06_l_capsule
   center=<-2.7286 -0.1519 -0.2792>
   rotation=<rot -3.8986 -89.2453 -2.7074>
  }]
 dyn_tassels_06_r [{
   shapeType=Capsule
   length=1.3848
   radius=0.9084
   shapeName=dyn_tassels_06_r_capsule
   center=<2.7286 0.1519 0.2792>
   rotation=<rot -3.8986 -89.2453 177.2926>
  }]
 dyn_tassels_07_l [{
   shapeType=Capsule
   length=0.9992
   radius=1.0901
   shapeName=dyn_tassels_07_l_capsule
   center=<-2.8105 -0.4516 0.1239>
   rotation=<rot 6.9108 -90.0001 10.7172>
  }]
 dyn_tassels_07_r [{
   shapeType=Capsule
   length=0.9992
   radius=1.0901
   shapeName=dyn_tassels_07_r_capsule
   center=<2.8105 0.4516 -0.1239>
   rotation=<rot 6.9108 -90.0001 -169.2828>
  }]
 dyn_tassels_08_l [{
   shapeType=Capsule
   length=1.2762
   radius=0.9495
   shapeName=dyn_tassels_08_l_capsule
   center=<-3.3639 -0.1227 0.1084>
   rotation=<rot -8.7898 -90.0015 7.9559>
  }]
 dyn_tassels_08_r [{
   shapeType=Capsule
   length=1.2762
   radius=0.9495
   shapeName=dyn_tassels_08_r_capsule
   center=<3.3639 0.1227 -0.1084>
   rotation=<rot -8.7898 -90.0015 -172.0441>
  }]
 dyn_tassels_main_l [{shapeType=Capsule, length=0.7948, radius=0.5088, shapeName=base_tassels_main_l_capsule, center=<-3.9135 -0.3271 0>, rotation=<rot 0 -90 0>}]
 dyn_tassels_main_r [{shapeType=Capsule, length=0.7948, radius=0.5088, shapeName=base_tassels_main_r_capsule, center=<3.9135 0.3271 0>, rotation=<rot 0 -90 180>}]
 upperarm_twist_01_l [{
   shapeType=Capsule
   length=10.0924
   radius=4.4372
   shapeName=upperarm_twist_01_l_capsule
   center=<1.2544 0.4365 -1.308>
   rotation=<rot -0.0008 85.8788 -89.9076>
  }]
 upperarm_twist_01_r [{
   shapeType=Capsule
   length=10.0924
   radius=4.4372
   shapeName=upperarm_twist_01_r_capsule
   center=<-1.2544 -0.4365 1.308>
   rotation=<rot -0.0008 85.8788 90.0924>
  }]
}
constraints [
 {bone1Name=dyn_tassels_01_l, bone2Name=dyn_tassels_main_l, swing1LimitDegrees=40, swing2LimitDegrees=15, twistLimitDegrees=10}
 {bone1Name=dyn_tassels_02_l, bone2Name=dyn_tassels_main_l, swing1LimitDegrees=50, swing2LimitDegrees=15, twistLimitDegrees=10}
 {bone1Name=dyn_tassels_03_l, bone2Name=dyn_tassels_main_l, swing1LimitDegrees=50, swing2LimitDegrees=15, twistLimitDegrees=10}
 {bone1Name=dyn_tassels_04_l, bone2Name=dyn_tassels_main_l, swing1LimitDegrees=50, swing2LimitDegrees=15, twistLimitDegrees=10}
 {bone1Name=dyn_tassels_05_l, bone2Name=dyn_tassels_main_l, swing1LimitDegrees=50, swing2LimitDegrees=15, twistLimitDegrees=10}
 {bone1Name=dyn_tassels_06_l, bone2Name=dyn_tassels_main_l, swing1LimitDegrees=50, swing2LimitDegrees=15, twistLimitDegrees=10}
 {bone1Name=dyn_tassels_07_l, bone2Name=dyn_tassels_main_l, swing1LimitDegrees=50, swing2LimitDegrees=15, twistLimitDegrees=10}
 {bone1Name=dyn_tassels_08_l, bone2Name=dyn_tassels_main_l, swing1LimitDegrees=50, swing2LimitDegrees=15, twistLimitDegrees=10}
 {bone1Name=dyn_tassels_main_l, bone2Name=base_tassels_main_l, swing1LimitDegrees=8, swing2LimitDegrees=4, twistLimitDegrees=3}
 {bone1Name=dyn_tassels_main_r, bone2Name=base_tassels_main_r, swing1LimitDegrees=4, swing2LimitDegrees=8, twistLimitDegrees=3}
 {bone1Name=dyn_tassels_01_r, bone2Name=dyn_tassels_main_r, swing1LimitDegrees=50, swing2LimitDegrees=15, twistLimitDegrees=10}
 {bone1Name=dyn_tassels_02_r, bone2Name=dyn_tassels_main_r, swing1LimitDegrees=50, swing2LimitDegrees=15, twistLimitDegrees=10}
 {bone1Name=dyn_tassels_03_r, bone2Name=dyn_tassels_main_r, swing1LimitDegrees=50, swing2LimitDegrees=15, twistLimitDegrees=10}
 {bone1Name=dyn_tassels_04_r, bone2Name=dyn_tassels_main_r, swing1LimitDegrees=50, swing2LimitDegrees=15, twistLimitDegrees=10}
 {bone1Name=dyn_tassels_05_r, bone2Name=dyn_tassels_main_r, swing1LimitDegrees=50, swing2LimitDegrees=15, twistLimitDegrees=10}
 {bone1Name=dyn_tassels_06_r, bone2Name=dyn_tassels_main_r, swing1LimitDegrees=50, swing2LimitDegrees=15, twistLimitDegrees=10}
 {bone1Name=dyn_tassels_07_r, bone2Name=dyn_tassels_main_r, swing1LimitDegrees=50, swing2LimitDegrees=15, twistLimitDegrees=10}
 {bone1Name=dyn_tassels_08_r, bone2Name=dyn_tassels_main_r, swing1LimitDegrees=50, swing2LimitDegrees=15, twistLimitDegrees=10}
]
