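# Post-process-style cosmetic AnimBP for the female QuirkLuster body, animating skirt, belts, pin and tassels over the parent mesh's copied pose.
# Control Rig and additive bone offsets feed AnimDynamics belts and pin, then RigidBody tassels, then one of two skirt simulations (standard or cape-equipped).
# EventGraph updates gravity, LOD, backbling and crouch flags, plus interpolated noise-driven flap rotators, zeroed for cape, DBNO, front end or disabled rigid bodies.
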AnimBlueprint /BRCosmetics/Characters/Player/Female/Medium/Bodies/F_MED_QuirkLuster/Meshes/F_MED_QuirkLuster_AnimBP
target_skeleton=&/F_MED_QuirkLuster_Skeleton
graphs [
 AnimGraph {nodes [
   RigidBody!(ComponentPose, Alpha, OverrideWorldGravity:Vector="0,0,0") -> (Pose) {
    bClampLinearTranslationLimitToRefPose=true
    bOverrideWorldGravity=true
    overridePhysicsAsset=&/F_MED_QuirkLuster_Physics_Skirt
    simulationSpace=BaseBoneSpace
    alphaBoolBlend {blendOption=Linear}
    baseBoneRef {boneName=pelvis}
    simSpaceSettings {
     maxAngularAcceleration=1600
     maxAngularVelocity=4
     maxLinearAcceleration=1000
     maxLinearVelocity=400
     worldAlpha=1
     externalLinearDragV=<1.1 1.1 1.1>
    }
   }
   n1 = "Transform (Modify) Bone - Bone: base_SkirtFlap_00_l"!(ComponentPose=Local To Component(LocalPose=Use cached pose 'DynamicsCache'()), bAlphaBoolEnabled:bool=True, Rotation="-5,0,0") -> (Pose) {
    alphaInputType=Bool
    rotationMode=BMM_Additive
    rotationSpace=BCS_BoneSpace
    alphaBoolBlend {blendOption=Linear}
    boneToModify {boneName=base_SkirtFlap_00_l}
    rotation {pitch=-5}
   }
   n2 = "Transform (Modify) Bone - Bone: base_SkirtFlap_00_r"!(ComponentPose=@n1, bAlphaBoolEnabled:bool=True, Rotation="-5,0,0") -> (Pose) {
    alphaInputType=Bool
    rotationMode=BMM_Additive
    rotationSpace=BCS_BoneSpace
    alphaBoolBlend {blendOption=Linear}
    boneToModify {boneName=base_SkirtFlap_00_r}
    rotation {pitch=-5}
   }
   Output Pose(Result=Component To Local(ComponentPose=@n2)) {layerGroup=None}
   n3 = Control Rig!(
    Source=Copy Pose From Mesh() {bUseAttachedParent=true}
    Alpha
    isKneeSliding:bool=false
    rBanEnabeld:bool=false
    startCrouching:bool=false
    isBaseballSliding:bool=false
    lowLODs:bool=false
    isSkyDiving:bool=True
   ) -> (Pose) {controlRigAssetReference {blueprintRigClass=&/F_MED_QuirkLuster_CtrlRig.F_MED_QuirkLuster_CtrlRig_C}}
   n4 = "Transform (Modify) Bone - Bone: attach_cape"!(ComponentPose=Local To Component(LocalPose=@n3), bAlphaBoolEnabled:bool=True, Translation="0,3,0", Rotation="0,0,-10") -> (Pose) {
    alphaInputType=Bool
    rotationMode=BMM_Additive
    rotationSpace=BCS_BoneSpace
    translationMode=BMM_Additive
    translationSpace=BCS_BoneSpace
    boneToModify {boneName=attach_cape}
    rotation {roll=-10}
    translation {y=3}
   }
   n5 = "Transform (Modify) Bone - Bone: dyn_beltPivotEnd"!(ComponentPose=@n4, bAlphaBoolEnabled:bool=True, Translation="-1,0,0", Rotation="0,-5,0") -> (Pose) {
    alphaInputType=Bool
    rotationMode=BMM_Additive
    rotationSpace=BCS_BoneSpace
    translationMode=BMM_Additive
    translationSpace=BCS_BoneSpace
    boneToModify {boneName=dyn_beltPivotEnd}
    rotation {yaw=-5}
    translation {x=-1}
   }
   Save cached pose 'CRCache_Input'(Pose=Component To Local(ComponentPose=@n5)) {cacheName=CRCache_Input}
   n6 = RigidBody(ComponentPose=Local To Component(LocalPose=Use cached pose 'preFlapsCache'()), OverrideWorldGravity="0,0,0") {
    bClampLinearTranslationLimitToRefPose=true
    bOverrideWorldGravity=true
    overridePhysicsAsset=&/F_MED_QuirkLuster_Physics_Skirt_CapeEquipped
    simulationSpace=BaseBoneSpace
    alphaBoolBlend {blendOption=Linear}
    baseBoneRef {boneName=pelvis}
    simSpaceSettings {
     maxAngularAcceleration=1600
     maxAngularVelocity=4
     maxLinearAcceleration=1000
     maxLinearVelocity=400
     velocityScaleZ=0.5
     worldAlpha=1
     externalLinearDragV=<1.2 1.2 1.2>
    }
   }
   n7 = RigidBody(ComponentPose=Local To Component(LocalPose=Use cached pose 'preFlapsCache'()), OverrideWorldGravity="0,0,0") {
    bClampLinearTranslationLimitToRefPose=true
    bOverrideWorldGravity=true
    overridePhysicsAsset=&F_MED_QuirkLuster_Physics_Skirt
    simulationSpace=BaseBoneSpace
    alphaBoolBlend {blendOption=Linear}
    baseBoneRef {boneName=pelvis}
    simSpaceSettings {
     maxAngularAcceleration=1600
     maxAngularVelocity=4
     maxLinearAcceleration=1000
     maxLinearVelocity=400
     worldAlpha=1
     externalLinearDragV=<1.2 1.2 1.2>
    }
   }
   Save cached pose 'DynamicsCache'(Pose=Blend Poses by bool(BlendPose_0=Component To Local(ComponentPose=@n6), BlendPose_1=Component To Local(ComponentPose=@n7))) {cacheName=DynamicsCache}
   n8 = "Transform (Modify) Bone - Bone: base_SkirtFlap_01_l"!(ComponentPose=Local To Component(LocalPose=Use cached pose 'CRCache_Input'()), Alpha=1, Rotation="0,0,0") -> (Pose) {rotationMode=BMM_Additive, rotationSpace=BCS_BoneSpace, alphaBoolBlend {blendOption=Linear}, boneToModify {boneName=base_SkirtFlap_01_l}}
   n9 = "Transform (Modify) Bone - Bone: base_SkirtFlap_01_r"!(ComponentPose=@n8, Alpha=1, Rotation="0,0,0") -> (Pose) {rotationMode=BMM_Additive, rotationSpace=BCS_BoneSpace, alphaBoolBlend {blendOption=Linear}, boneToModify {boneName=base_SkirtFlap_01_r}}
   Save cached pose 'preDynamicsCache'(Pose=Component To Local(ComponentPose=@n9)) {cacheName=preDynamicsCache}
   n10 = Anim Dynamics - Bone: dyn_beltPivotMid!(ComponentPose=Local To Component(LocalPose=Use cached pose 'preDynamicsCache'()), Alpha=0.5, GravityOverride:Vector="0,0,-980") -> (Pose) {
    angularDampingOverride=1
    angularSpringConstant=10000
    bAngularSpring=true
    bOverrideAngularDamping=true
    bUseGravityOverride=true
    gravityScale=0.35
    simulationSpace=RootRelative
    boundBone {boneName=dyn_beltPivotMid}
    componentAppliedLinearAccClamp=<5000 5000 5000>
    componentLinearAccScale=<1 1 1>
    componentLinearVelScale=<1 1 1>
    gravityOverride {z=-980}
    physicsBodyDefinitions [{
      boundBone {boneName=dyn_beltPivotMid}
      boxExtents=<20 30 30>
      constraintSetup {angularLimitsMax=<0 0 0>, angularLimitsMin=<0 -15 0>, angularTarget=<1 0 0>}
      localJointOffset=<0.4 0 0>
     }]
    simSpaceSettings {maxAngularAcceleration=20, maxAngularVelocity=4, simSpaceAngularAlpha=1}
   }
   n11 = Anim Dynamics - Bone: dyn_beltPivotEnd!(ComponentPose=@n10, Alpha=0.5, GravityOverride:Vector="0,0,-980") -> (Pose) {
    angularDampingOverride=1
    angularSpringConstant=10000
    bAngularSpring=true
    bOverrideAngularDamping=true
    bUseGravityOverride=true
    gravityScale=0.35
    simulationSpace=RootRelative
    boundBone {boneName=dyn_beltPivotEnd}
    componentAppliedLinearAccClamp=<5000 5000 5000>
    componentLinearAccScale=<1 1 1>
    componentLinearVelScale=<1 1 1>
    gravityOverride {z=-980}
    physicsBodyDefinitions [{
      boundBone {boneName=dyn_beltPivotEnd}
      boxExtents=<20 30 30>
      constraintSetup {angularLimitsMax=<0 0 0>, angularLimitsMin=<0 -10 0>, angularTarget=<1 0 0>}
      localJointOffset=<0.4 0 0>
     }]
    simSpaceSettings {maxAngularAcceleration=20, maxAngularVelocity=4, simSpaceAngularAlpha=1}
   }
   n12 = Anim Dynamics - Bone: dyn_upperBeltStrap_start!(ComponentPose=@n11, Alpha=0.5, GravityOverride:Vector="0,0,-980") -> (Pose) {
    angularDampingOverride=1
    angularSpringConstant=10000
    bAngularSpring=true
    bOverrideAngularDamping=true
    bUseGravityOverride=true
    gravityScale=0.35
    simulationSpace=RootRelative
    boundBone {boneName=dyn_upperBeltStrap_start}
    componentAppliedLinearAccClamp=<5000 5000 5000>
    componentLinearAccScale=<1 1 1>
    componentLinearVelScale=<1 1 1>
    gravityOverride {z=-980}
    physicsBodyDefinitions [{
      boundBone {boneName=dyn_upperBeltStrap_start}
      boxExtents=<20 30 30>
      constraintSetup {angularLimitsMax=<0 10 0>, angularLimitsMin=<0 -10 0>, angularTarget=<1 0 0>}
      localJointOffset=<0.4 0 0>
     }]
    simSpaceSettings {maxAngularAcceleration=20, maxAngularVelocity=4, simSpaceAngularAlpha=1}
   }
   n13 = Anim Dynamics - Bone: dyn_pin!(ComponentPose=@n12, Alpha=1, GravityOverride:Vector="0,0,-980") -> (Pose) {
    angularDampingOverride=1
    angularSpringConstant=1000
    bAngularSpring=true
    bOverrideAngularDamping=true
    bUseGravityOverride=true
    gravityScale=0.35
    simulationSpace=RootRelative
    boundBone {boneName=dyn_pin}
    componentAppliedLinearAccClamp=<5000 5000 5000>
    componentLinearAccScale=<1 1 1>
    componentLinearVelScale=<1 1 1>
    gravityOverride {z=-980}
    physicsBodyDefinitions [{
      boundBone {boneName=dyn_pin}
      boxExtents=<8 12 12>
      constraintSetup {angularLimitsMax=<0 30 90>, angularLimitsMin=<0 -30 0>, angularTarget=<1 0 0>}
      localJointOffset=<0.5 0 0>
     }]
    simSpaceSettings {maxAngularAcceleration=20, maxAngularVelocity=5, simSpaceAngularAlpha=1}
   }
   n14 = RigidBody(ComponentPose=@n13, OverrideWorldGravity="0,0,0") {
    bOverrideWorldGravity=true
    overridePhysicsAsset=&/F_MED_QuirkLuster_Physics_Tassles
    simulationSpace=BaseBoneSpace
    alphaBoolBlend {blendOption=Linear}
    baseBoneRef {boneName=spine_05}
    simSpaceSettings {
     dampingAlpha=0
     maxAngularAcceleration=800
     maxAngularVelocity=8
     maxLinearAcceleration=1200
     maxLinearVelocity=700
     velocityScaleZ=0.3
     worldAlpha=1
     externalLinearDragV=<1 1 1>
    }
   }
   Save cached pose 'preFlapsCache'(Pose=Component To Local(ComponentPose=@n14)) {cacheName=preFlapsCache}
  ]}
 EventGraph {nodes [
   n1 = Event BlueprintUpdateAnimation()
   > Set Delta Time X(Delta Time X=@n1.DeltaTimeX)
   > n2 = Sequence!(execute) -> (then_0, then_1, then_2:Exec, then_3:Exec, then_4:Exec)
   Branch(execute=@n2.then_4, Condition=Get bIsInFrontEnd())
   n3 = Get bIsSkyDiving()
   n4 = Get bIsPlayingEmote()
   > Set backBlingFE(backBlingFE:bool=AND Boolean(A=NOT Boolean(A=@n4), B=NOT Boolean(A=@n3))) -> (Output_Get:bool)
   n5 = "Gravity Override State Function (Parameter Asset) (Nativized)"(
    execute=@n2.then_0
    CharacterPartAnimInstance=Self-Reference()
    GameTimeInSeconds=GetGameTimeInSeconds()
    Parameters=&/F_MED_QuirkLuster_Parameters
   )
   > n6 = Set Gravity Override(Gravity Override=@n5)
   > Branch(Condition="Equal (Vector)"!(A:Vector=@n6, B:Vector, ErrorTolerance:float) -> (ReturnValue:bool=~) {is=title K2Node_PromotableOperator})
   > Set Gravity Override(Gravity Override="0,0,-400")
   Set LowLODs(execute=@n2.then_2, LowLODs:bool="integer < integer"(A=Get CurrentLOD(), B=2)) -> (Output_Get:bool)
   n7 = Get bIsInFrontEnd()
   n8 = Get bIsDBNO()
   n9 = Get isWearingCape() -> (isWearingCape:bool)
   n10 = Branch(
    execute=@n2.then_1
    Condition=AND Boolean!(A=Get bRigidBodyEnabled(), B=NOT Boolean(A=@n8), C:bool=NOT Boolean(A=@n7), D:bool=NOT Boolean(A=@n9)) -> (ReturnValue)
   )
   > n11 = RandomRotatorFunction!(
    execute
    GameTimeInSeconds:double=GetGameTimeInSeconds()
    PerlinScale:double=1
    Frequency:Rotator="4,1,0"
    Amplitude:Rotator="35,5,0"
    Offset:Rotator="12,0,0"
    TimeOffset:double=0
   ) -> (then, RandomRotator:Rotator)
   > Set FlapRotator_l(FlapRotator_l:Rotator=RInterpTo!(
     Current:Rotator=Get FlapRotator_l() -> (FlapRotator_l:Rotator)
     Target:Rotator=@n11
     DeltaTime:float=Get Delta Time X()
     InterpSpeed:float=2.5
    ) -> (ReturnValue:Rotator)) -> (Output_Get:Rotator)
   > n12 = RandomRotatorFunction(GameTimeInSeconds=GetGameTimeInSeconds(), PerlinScale=1, Frequency="3.5,2.5,0", Amplitude="-30,5,0", Offset="12,0,0", TimeOffset=20)
   > Set FlapRotator_r(FlapRotator_r:Rotator=RInterpTo(Current=Get FlapRotator_r() -> (FlapRotator_r:Rotator), Target=@n12, DeltaTime=Get Delta Time X(), InterpSpeed=3)) -> (Output_Get:Rotator)
   Set FlapRotator_l(execute=@n10.else, FlapRotator_l="0, 0, 0")
   > Set FlapRotator_r(FlapRotator_r="0, 0, 0")
   Set isWearingCape(execute=@n2.then_3, isWearingCape=CheckForBackbling().bIsBackblingACape)
   Event OnBeginCrouching()
   > Set startCrouch(startCrouch:bool=NOT Boolean(A=Get isWearingCape())) -> (Output_Get:bool)
   > Delay(Duration=0.2)
   > Set startCrouch(startCrouch=false)
   Event OnEndCrouching()
   > Set startCrouch(startCrouch=false)
  ]}
]
variables [
 backBlingFE {type=boolean}
 Delta Time S {type=number}
 Delta Time X {type=number}
 FlapRotator_l {type=Rotator}
 FlapRotator_r {type=Rotator}
 Gravity Override {type=Vector, default=<0 0 -980>}
 isWearingCape {type=boolean}
 LowLODs {type=boolean}
 startCrouch {type=boolean}
]
