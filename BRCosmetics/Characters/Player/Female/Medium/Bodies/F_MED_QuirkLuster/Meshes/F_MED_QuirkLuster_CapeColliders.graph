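# Physics asset of capsule colliders for the QuirkLuster female medium cosmetic body, named as the collision set for its cape.
# Capsules cover head, clavicles, spine, pelvis and feet, while arm, thigh, calf, neck and shoulder-pad bodies are empty.
# Left and right five-bone skirt-flap chains carry capsules, mostly overlapping in groups to approximate each flap's surface.
# Constraints link each flap chain to the pelvis and shoulder pads to upper arms, with all swing and twist free.
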PhysicsAsset /BRCosmetics/Characters/Player/Female/Medium/Bodies/F_MED_QuirkLuster/Meshes/F_MED_QuirkLuster_CapeColliders
bodies {
 calf_l []
 calf_r []
 clavicle_l [{shapeType=Capsule, length=6.5623, radius=7.5194, center=<3.9566 -1.0862 -2.1127>, rotation=<rot -77.5566 177.3915 7.7314>}]
 clavicle_r [{shapeType=Capsule, length=6.5623, radius=7.5194, center=<-3.9566 1.0862 2.1127>, rotation=<rot -77.5566 177.3915 -172.2686>}]
 dyn_SkirtFlap_00_l [{
   shapeType=Capsule
   length=12.4996
   radius=5.6273
   shapeName=dyn_SkirtFlap_00_l_capsule
   center=<5.9047 -0.8838 -3.8345>
   rotation=<rot -0.4259 -0.8233 -74.7099>
  }]
 dyn_SkirtFlap_00_r [{
   shapeType=Capsule
   length=5.5896
   radius=5.6392
   shapeName=dyn_SkirtFlap_00_r_capsule
   center=<5.897 2.0131 -3.8794>
   rotation=<rot 0 -177.4324 -72.0153>
  }]
 dyn_SkirtFlap_01_l [
  {
   shapeType=Capsule
   length=14.1316
   radius=7.7405
   shapeName=dyn_SkirtFlap_01_l_capsule
   center=<10.665 0.0979 -4.0981>
   rotation=<rot -0.1699 1.2912 -82.8>
  }
  {
   shapeType=Capsule
   length=14.9085
   radius=7.0733
   shapeName=dyn_SkirtFlap_01_l_capsule
   center=<5.5821 0.2152 -4.2214>
   rotation=<rot -0.1699 1.2912 -82.8>
  }
  {
   shapeType=Capsule
   length=13.9429
   radius=8.0389
   shapeName=dyn_SkirtFlap_01_l_capsule
   center=<15.0802 0.4293 -4.2496>
   rotation=<rot -0.1699 1.2912 -82.8>
  }
 ]
 dyn_SkirtFlap_01_r [
  {
   shapeType=Capsule
   length=14.1162
   radius=7.756
   shapeName=dyn_SkirtFlap_01_r_capsule
   center=<10.6308 0.6847 -4.0725>
   rotation=<rot -0.1699 -1.2912 -97.2>
  }
  {
   shapeType=Capsule
   length=14.8931
   radius=7.0887
   shapeName=dyn_SkirtFlap_01_r_capsule
   center=<5.5541 0.7946 -4.2202>
   rotation=<rot -0.1699 -1.2912 -97.2>
  }
  {
   shapeType=Capsule
   length=14.3562
   radius=7.6256
   shapeName=dyn_SkirtFlap_01_r_capsule
   center=<14.0052 0.6041 -4.2453>
   rotation=<rot -0.1699 -1.2912 -97.2>
  }
 ]
 dyn_SkirtFlap_02_l [
  {
   shapeType=Capsule
   length=16.4873
   radius=8.5641
   shapeName=dyn_SkirtFlap_01_l_capsule
   center=<12.4715 1.8998 -4.3478>
   rotation=<rot -0.17 1.2912 -78.8391>
  }
  {
   shapeType=Capsule
   length=16.9868
   radius=8.1249
   shapeName=dyn_SkirtFlap_01_l_capsule
   center=<5.5596 1.1951 -4.3486>
   rotation=<rot -0.17 1.2912 -78.8391>
  }
 ]
 dyn_SkirtFlap_02_r [
  {
   shapeType=Capsule
   length=16.5915
   radius=8.7432
   shapeName=dyn_SkirtFlap_01_r_capsule
   center=<12.6458 -0.3819 -4.2711>
   rotation=<rot -0.17 -1.2912 -101.1609>
  }
  {
   shapeType=Capsule
   length=17.0307
   radius=8.3845
   shapeName=dyn_SkirtFlap_01_r_capsule
   center=<5.5345 0.0669 -4.6348>
   rotation=<rot -0.17 -1.2912 -101.1609>
  }
 ]
 dyn_SkirtFlap_03_l [
  {
   shapeType=Capsule
   length=18.8617
   radius=8.2478
   shapeName=dyn_SkirtFlap_01_l_capsule
   center=<9.9896 2.5209 -4.5122>
   rotation=<rot -0.1701 1.2912 -75.6117>
  }
  {
   shapeType=Capsule
   length=17.3668
   radius=8.748
   shapeName=dyn_SkirtFlap_01_l_capsule
   center=<3.5663 2.2842 -4.5745>
   rotation=<rot -0.1701 1.2912 -75.6117>
  }
 ]
 dyn_SkirtFlap_03_r [
  {
   shapeType=Capsule
   length=17.561
   radius=8.3931
   shapeName=dyn_SkirtFlap_01_r_capsule
   center=<10.1947 -0.6334 -4.5634>
   rotation=<rot -0.1701 -1.2912 -104.3883>
  }
  {
   shapeType=Capsule
   length=18.7257
   radius=8.5398
   shapeName=dyn_SkirtFlap_01_r_capsule
   center=<3.4043 -0.8429 -4.3512>
   rotation=<rot -0.1701 -1.2912 -104.3883>
  }
 ]
 dyn_SkirtFlap_04_l [
  {
   shapeType=Capsule
   length=2.5393
   radius=7.5615
   shapeName=dyn_SkirtFlap_01_l_capsule
   center=<7.2007 -0.5867 -4.3819>
   rotation=<rot -9.5889 41.8077 -82.0942>
  }
  {
   shapeType=Capsule
   length=12.7256
   radius=7.8787
   shapeName=dyn_SkirtFlap_01_l_capsule
   center=<3.6869 1.0171 -4.2498>
   rotation=<rot -0.1701 1.2912 -75.6117>
  }
  {
   shapeType=Capsule
   length=2.8829
   radius=7.3276
   shapeName=dyn_SkirtFlap_01_l_capsule
   center=<9.7246 -3.4873 -4.0701>
   rotation=<rot -14.3782 88.3342 -89.4312>
  }
 ]
 dyn_SkirtFlap_04_r [
  {
   shapeType=Capsule
   length=2.5393
   radius=7.5615
   shapeName=dyn_SkirtFlap_01_r_capsule
   center=<7.1727 1.7238 -4.5584>
   rotation=<rot -9.5889 -41.8077 -97.9058>
  }
  {
   shapeType=Capsule
   length=12.6916
   radius=7.9126
   shapeName=dyn_SkirtFlap_01_r_capsule
   center=<3.2217 0.1297 -4.6522>
   rotation=<rot -0.1701 -1.2912 -104.3883>
  }
  {
   shapeType=Capsule
   length=3.6659
   radius=6.0711
   shapeName=dyn_SkirtFlap_01_r_capsule
   center=<9.6966 4.6243 -4.8333>
   rotation=<rot -14.3782 -88.3342 -90.5688>
  }
 ]
 dyn_shoulderPad_top_l []
 dyn_shoulderPad_top_r []
 foot_l [{shapeType=Capsule, length=11.0765, radius=8.0344, center=<-4.1811 6.0027 -0.338>, rotation=<rot 0.0388 1.4701 -89.3982>}]
 foot_r [{shapeType=Capsule, length=11.0765, radius=8.0344, center=<4.1811 -6.0027 0.338>, rotation=<rot 0.0388 1.4701 90.6018>}]
 head [{shapeType=Capsule, length=0, radius=18, center=<9.0697 0.0746 0>, rotation=<rot 0 -90 0>}]
 lowerarm_l []
 lowerarm_r []
 neck_02 []
 pelvis [{shapeType=Capsule, length=9.6042, radius=11.1441, center=<-2.4471 -2.5289 -0.1981>}]
 spine_01 [{shapeType=Capsule, length=4.3888, radius=14.3638, center=<-8.9021 -0.8973 0>}]
 spine_02 [
  {shapeType=Capsule, length=6.3819, radius=9.7317, center=<-2.6885 -0.5847 0>}
  {shapeType=Capsule, length=20.6875, radius=9.3125, center=<-1.8317 1.9362 0>, rotation=<rot 90 0 0>}
 ]
 spine_03 [
  {shapeType=Capsule, length=6.1863, radius=8.4854, center=<0.0574 0.4416 0>}
  {shapeType=Capsule, length=20.7188, radius=9.2812, center=<-1.2163 0.549 0>, rotation=<rot 90 180 -175.0732>}
 ]
 spine_04 [
  {shapeType=Capsule, length=9.0076, radius=10.3916, center=<5.6234 -0.6779 0>}
  {shapeType=Capsule, length=10.0312, radius=9.2812, center=<-1.2326 1.308 0>, rotation=<rot 90 0 0>}
 ]
 spine_05 [{shapeType=Capsule, length=8.9485, radius=12.3571, center=<0.6969 1.91 0>}]
 thigh_l []
 thigh_r []
 upperarm_l []
 upperarm_r []
}
constraints [
 {
  bone1Name=dyn_shoulderPad_top_l
  bone2Name=upperarm_l
  swing1LimitDegrees=45
  swing1Motion=Free
  swing2LimitDegrees=45
  swing2Motion=Free
  twistLimitDegrees=45
  twistMotion=Free
 }
 {
  bone1Name=dyn_shoulderPad_top_r
  bone2Name=upperarm_r
  swing1LimitDegrees=45
  swing1Motion=Free
  swing2LimitDegrees=45
  swing2Motion=Free
  twistLimitDegrees=45
  twistMotion=Free
 }
 {
  bone1Name=dyn_SkirtFlap_00_l
  bone2Name=pelvis
  swing1LimitDegrees=45
  swing1Motion=Free
  swing2LimitDegrees=45
  swing2Motion=Free
  twistLimitDegrees=45
  twistMotion=Free
 }
 {
  bone1Name=dyn_SkirtFlap_01_l
  bone2Name=dyn_SkirtFlap_00_l
  swing1LimitDegrees=45
  swing1Motion=Free
  swing2LimitDegrees=45
  swing2Motion=Free
  twistLimitDegrees=45
  twistMotion=Free
 }
 {
  bone1Name=dyn_SkirtFlap_02_l
  bone2Name=dyn_SkirtFlap_01_l
  swing1LimitDegrees=45
  swing1Motion=Free
  swing2LimitDegrees=45
  swing2Motion=Free
  twistLimitDegrees=45
  twistMotion=Free
 }
 {
  bone1Name=dyn_SkirtFlap_03_l
  bone2Name=dyn_SkirtFlap_02_l
  swing1LimitDegrees=45
  swing1Motion=Free
  swing2LimitDegrees=45
  swing2Motion=Free
  twistLimitDegrees=45
  twistMotion=Free
 }
 {
  bone1Name=dyn_SkirtFlap_04_l
  bone2Name=dyn_SkirtFlap_03_l
  swing1LimitDegrees=45
  swing1Motion=Free
  swing2LimitDegrees=45
  swing2Motion=Free
  twistLimitDegrees=45
  twistMotion=Free
 }
 {
  bone1Name=dyn_SkirtFlap_00_r
  bone2Name=pelvis
  swing1LimitDegrees=45
  swing1Motion=Free
  swing2LimitDegrees=45
  swing2Motion=Free
  twistLimitDegrees=45
  twistMotion=Free
 }
 {
  bone1Name=dyn_SkirtFlap_01_r
  bone2Name=dyn_SkirtFlap_00_r
  swing1LimitDegrees=45
  swing1Motion=Free
  swing2LimitDegrees=45
  swing2Motion=Free
  twistLimitDegrees=45
  twistMotion=Free
 }
 {
  bone1Name=dyn_SkirtFlap_02_r
  bone2Name=dyn_SkirtFlap_01_r
  swing1LimitDegrees=45
  swing1Motion=Free
  swing2LimitDegrees=45
  swing2Motion=Free
  twistLimitDegrees=45
  twistMotion=Free
 }
 {
  bone1Name=dyn_SkirtFlap_03_r
  bone2Name=dyn_SkirtFlap_02_r
  swing1LimitDegrees=45
  swing1Motion=Free
  swing2LimitDegrees=45
  swing2Motion=Free
  twistLimitDegrees=45
  twistMotion=Free
 }
 {
  bone1Name=dyn_SkirtFlap_04_r
  bone2Name=dyn_SkirtFlap_03_r
  swing1LimitDegrees=45
  swing1Motion=Free
  swing2LimitDegrees=45
  swing2Motion=Free
  twistLimitDegrees=45
  twistMotion=Free
 }
]
