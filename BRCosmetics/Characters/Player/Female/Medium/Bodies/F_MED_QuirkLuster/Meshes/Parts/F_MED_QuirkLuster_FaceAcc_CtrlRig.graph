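# Control Rig for the F_MED_QuirkLuster face accessory part, procedurally correcting hair and belt-hung sword bones against body poses.
# Constraints tie the sword to belt and pelvis, then head look-up, tilt and pelvis-height crouch alphas are computed.
# A swing/twist-to-alpha function and spherical pose readers on right arm and left thigh weight additive bone offsets, some smoothed.
# An eased skydiving hair pose runs last; FK controls are optional.
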ControlRigBlueprint /BRCosmetics/Characters/Player/Female/Medium/Bodies/F_MED_QuirkLuster/Meshes/Parts/F_MED_QuirkLuster_FaceAcc_CtrlRig
graphs [
 CollapseNode_ContainedGraph {nodes [
   n1 = RigVMFunctionEntryNode()
   &RigUnit_ParentConstraint(
    AdvancedSettings=(InterpolationType=Average,RotationOrderForFilter=XZY)
    Child=(Type=Bone,Name="base_sword")
    Parents=((Item=(Type=Bone,Name="dyn_beltPivotEnd"),Weight=1))
    Weight=1
    ExecutePin=@n1.ExecutePin
   )
   > n2 = &RigUnit_RotationConstraintLocalSpaceOffset(
    AdvancedSettings=(InterpolationType=Average,RotationOrderForFilter=XZY)
    Child=(Type=Bone,Name="base_sword")
    Filter=(bX=True,bY=True,bZ=True)
    Parents=((Item=(Type=Bone,Name="pelvis"),Weight=1))
    Weight=1
   )
   RigVMFunctionReturnNode(ExecutePin=@n2.ExecutePin)
  ]}
 CollapseNode_ContainedGraph {nodes [
   RigVMFunctionEntryNode()
   > GetAlphaFromQuatDelta = RigVMFunctionReferenceNode(
    Rotation Order=ZYX
    child=(Type=Bone,Name="head")
    debugRotation=false
    outputMinMax=(X=0,Y=1)
    parent=(Type=Bone,Name="neck_01")
    premult=1
    rotationMinMax=(X=-20,Y=-43)
    triggerAxis=Z
    twistAxis=(X=1,Y=0,Z=0)
   )
   n1 = &RigVMFunction_AnimEvalRichCurve(
    Curve="(EditorCurveData=(Keys=((InterpMode=RCIM_Cubic,TangentMode=RCTM_User,ArriveTangent=0.6623,LeaveTangent=0.6623),(InterpMode=RCIM_Cubic,TangentMode=RCTM_User,Time=0.5,Value=0.4,ArriveTangent=0.8898,LeaveTangent=0.8898),(InterpMode=RCIM_Cubic,TangentMode=RCTM ...(truncated)"
    SourceMinimum=0
    TargetMinimum=0
    Value=@GetAlphaFromQuatDelta.alpha
   )
   > RigVMVariableNode(Value=@n1.Result) {variable=HeadLookingUp}
   > GetAlphaFromQuatDelta_1 = RigVMFunctionReferenceNode(
    Rotation Order=ZYX
    child=(Type=Bone,Name="head")
    outputMinMax=(X=0,Y=1)
    parent=(Type=Bone,Name="neck_01")
    premult=1
    rotationMinMax=(X=-50,Y=-80)
    triggerAxis=Z
    twistAxis=(X=1,Y=0,Z=0)
   )
   > RigVMVariableNode(Value=@GetAlphaFromQuatDelta_1.alpha) {variable=HeadLookingUp_Extreme}
   > GetAlphaFromQuatDelta_2 = RigVMFunctionReferenceNode(
    Rotation Order=ZYX
    child=(Type=Bone,Name="head")
    outputMinMax=(X=0,Y=1)
    parent=(Type=Bone,Name="neck_01")
    premult=1
    rotationMinMax=(X=-50,Y=-65)
    triggerAxis=Z
    twistAxis=(X=1,Y=0,Z=0)
   )
   > RigVMVariableNode(Value=@GetAlphaFromQuatDelta_2.alpha) {variable=HeadLookingUp_Crouching}
   > GetAlphaFromQuatDelta_3 = RigVMFunctionReferenceNode(
    Rotation Order=ZYX
    child=(Type=Bone,Name="head")
    debugRotation=false
    outputMinMax=(X=0,Y=1)
    parent=(Type=Bone,Name="neck_01")
    premult=1
    rotationMinMax=(X=-20,Y=20)
    triggerAxis=Y
    twistAxis=(X=1,Y=0,Z=0)
   )
   > RigVMVariableNode(Value=@GetAlphaFromQuatDelta_3.alpha) {variable=HeadTiltingSideways}
   > RigVMVariableNode(Value=&RigVMFunction_MathDoubleRemap(
     SourceMaximum=65
     SourceMinimum=44
     TargetMaximum=0
     TargetMinimum=1
     Value=&RigUnit_GetRelativeTransformForItem(Child=(Type=Bone,Name="pelvis"), Parent=(Type=Bone,Name="root"), bChildInitial=False, bParentInitial=False).RelativeTransform.Translation.Z
    ).Result) {variable=isCrouchingInPlace}
   > RigVMFunctionReturnNode()
  ]}
 CollapseNode_ContainedGraph {nodes [
   n1 = RigVMFunctionEntryNode()
   RigVMFunction_Sequence = &RigVMFunction_Sequence(ExecuteContext=@n1.ExecutePin)
   GetAlphaFromQuatDelta = RigVMFunctionReferenceNode(
    Rotation Order=ZYX
    child=(Type=Bone,Name="neck_01")
    debugRotation=false
    outputMinMax=(X=1,Y=0)
    parent=(Type=Null,Name="neck_01_null")
    premult=1
    rotationMinMax=(X=-8,Y=0)
    triggerAxis=Z
    twistAxis=(X=1,Y=0,Z=0)
    ExecuteContext=@RigVMFunction_Sequence.A
   )
   Multiply = &RigVMFunction_MathDoubleMul(A=@GetAlphaFromQuatDelta.alpha, B=RigVMVariableNode() {variable=HeadLookingUp}.Value)
   > &RigUnit_ModifyTransforms(
    ItemToModify="((Item=(Type=Bone,Name=\"base_Hair_Side_00_l\"),Transform=(Rotation=(X=0,Y=0,Z=-0.1305,W=0.9914),Translation=(X=0,Y=0,Z=0),Scale3D=(X=1,Y=1,Z=1))),(Item=(Type=Bone,Name=\"base_Hair_Side_00_r\"),Transform=(Rotation=(X=0,Y=0,Z=0.0436,W=0.999),Translation=(X=0,Y= ...(truncated)"
    WeightMinimum=0
    Weight=@Multiply.Result
   )
   > &RigUnit_ModifyTransforms(
    ItemToModify="((Item=(Type=Bone,Name=\"base_Hair_Side_00_l\"),Transform=(Rotation=(X=-0.0076,Y=-0.043,Z=-0.1735,W=0.9839),Translation=(X=0,Y=0,Z=0),Scale3D=(X=1,Y=1,Z=1))),(Item=(Type=Bone,Name=\"base_Hair_Side_00_r\"),Transform=(Rotation=(X=0,Y=0,Z=0.1908,W=0.9816),Transla ...(truncated)"
    WeightMinimum=0
    Weight=RigVMVariableNode() {variable=HeadLookingUp_Crouching}.Value
   )
   > &RigUnit_ModifyTransforms(
    ItemToModify="((Item=(Type=Bone,Name=\"base_Hair_Side_01_l\"),Transform=(Rotation=(X=0,Y=0,Z=-0.0944,W=0.9955),Translation=(X=0,Y=0,Z=0),Scale3D=(X=1,Y=1,Z=1))),(Item=(Type=Bone,Name=\"base_Hair_Side_01_r\"),Transform=(Rotation=(X=-0.0038,Y=0.0435,Z=0.0871,W=0.9952),Transla ...(truncated)"
    WeightMinimum=0
    Weight=RigVMVariableNode() {variable=HeadLookingUp}.Value
   )
   Remap = &RigVMFunction_MathDoubleRemap(SourceMaximum=1, SourceMinimum=0.5, TargetMaximum=1, TargetMinimum=0, Value=RigVMVariableNode() {variable=HeadTiltingSideways}.Value)
   > &RigUnit_ModifyTransforms(
    ItemToModify="((Item=(Type=Bone,Name=\"base_Hair_Side_01_l\"),Transform=(Rotation=(X=0,Y=0,Z=0.0872,W=0.9962),Translation=(X=0,Y=0,Z=0),Scale3D=(X=1,Y=1,Z=1))),(Item=(Type=Bone,Name=\"base_HairStrand_Side_A_01_r\"),Transform=(Rotation=(X=0,Y=0.1305,Z=0,W=0.9914),Translation ...(truncated)"
    WeightMinimum=0
    Weight=@Remap.Result
   )
   > &RigUnit_ModifyTransforms(
    ItemToModify="((Item=(Type=Bone,Name=\"base_Hair_Side_01_r\"),Transform=(Rotation=(X=0,Y=0,Z=0.1305,W=0.9914),Translation=(X=0,Y=0,Z=0),Scale3D=(X=1,Y=1,Z=1))),(Item=(Type=Bone,Name=\"base_HairStrand_Side_A_01_r\"),Transform=(Rotation=(X=-0.0094,Y=0.0426,Z=0.2162,W=0.9754), ...(truncated)"
    WeightMinimum=0
    Weight=&RigVMFunction_MathDoubleRemap(SourceMaximum=0.5, SourceMinimum=0, TargetMaximum=0, TargetMinimum=1, Value=RigVMVariableNode() {variable=HeadTiltingSideways}.Value).Result
   )
   > n2 = &RigUnit_SphericalPoseReader(
    ActiveRegionScaleFactors=(PositiveWidth=0.5,NegativeWidth=0.5,PositiveHeight=0.5,NegativeHeight=0.5)
    ActiveRegionSize=0.1
    Debug=(bDrawDebug=false,bDraw2D=False,bDrawLocalAxes=False,DebugScale=25,DebugSegments=20,DebugThickness=0.25)
    DriverAxis=(X=-1,Y=0,Z=0)
    DriverItem=(Type=Bone,Name="upperarm_r")
    FalloffRegionScaleFactors=(PositiveWidth=1,NegativeWidth=1,PositiveHeight=1,NegativeHeight=1)
    FalloffSize=0.2
    FlipHeightScaling=False
    FlipWidthScaling=False
    OptionalParentItem=(Type=Bone,Name="clavicle_r")
    RotationOffset=(X=120,Y=0,Z=-140)
   )
   AlphaInterp = &RigVMFunction_AlphaInterp(InterpSpeedDecreasing=8, InterpSpeedIncreasing=14, Scale=1, bClampResult=False, Value=@n2.OutputParam)
   > &RigUnit_ModifyTransforms(
    ItemToModify="((Item=(Type=Bone,Name=\"base_Hair_Side_01_r\"),Transform=(Rotation=(X=0.1677,Y=0.2549,Z=-0.0449,W=0.9513),Translation=(X=-1,Y=0,Z=-2),Scale3D=(X=1,Y=1,Z=1))),(Item=(Type=Bone,Name=\"dyn_Hair_Side_02_r\"),Transform=(Rotation=(X=0,Y=-0.342,Z=0,W=0.9397),Transla ...(truncated)"
    WeightMinimum=0
    Weight=@AlphaInterp.Result
   )
   > n3 = &RigUnit_SphericalPoseReader(
    ActiveRegionScaleFactors=(PositiveWidth=0.5,NegativeWidth=0.5,PositiveHeight=1,NegativeHeight=0.5)
    ActiveRegionSize=0.05
    Debug=(bDrawDebug=false,bDraw2D=False,bDrawLocalAxes=False,DebugScale=25,DebugSegments=20,DebugThickness=0.25)
    DriverAxis=(X=-1,Y=0,Z=0)
    DriverItem=(Type=Bone,Name="upperarm_r")
    FalloffRegionScaleFactors=(PositiveWidth=1,NegativeWidth=1,PositiveHeight=1,NegativeHeight=1)
    FalloffSize=0.2
    FlipHeightScaling=False
    FlipWidthScaling=False
    OptionalParentItem=(Type=Bone,Name="clavicle_r")
    RotationOffset=(X=92,Y=0,Z=-154)
   )
   > &RigUnit_ModifyTransforms(
    ItemToModify="((Item=(Type=Bone,Name=\"base_HairStrand_Side_A_01_r\"),Transform=(Rotation=(X=0,Y=0.0872,Z=0,W=0.9962),Translation=(X=0,Y=0,Z=0),Scale3D=(X=1,Y=1,Z=1))),(Item=(Type=Bone,Name=\"base_Hair_Side_01_r\"),Transform=(Rotation=(X=0,Y=0.0872,Z=0,W=0.9962),Translation ...(truncated)"
    WeightMinimum=0
    Weight=@n3.OutputParam
   )
   > &RigUnit_ModifyTransforms(
    ItemToModify="((Item=(Type=Bone,Name=\"dyn_Hair_Back_02\"),Transform=(Rotation=(X=0,Y=0,Z=0,W=1),Translation=(X=-3,Y=0,Z=-1),Scale3D=(X=1,Y=1,Z=1))),(Item=(Type=Bone,Name=\"dyn_Hair_Mid_02_r\"),Transform=(Rotation=(X=0,Y=0,Z=0,W=1),Translation=(X=-3,Y=0,Z=0),Scale3D=(X=1,Y= ...(truncated)"
    WeightMinimum=0
    Weight=RigVMVariableNode() {variable=HeadLookingUp_Crouching}.Value
   )
   RigVMFunctionReturnNode(ExecutePin=@RigVMFunction_Sequence.B)
   &RigUnit_ModifyTransforms(
    ExecutePin=()
    ItemToModify="((Item=(Type=Bone,Name=\"base_HairStrand_Side_A_01_r\"),Transform=(Rotation=(X=0,Y=0,Z=0.3827,W=0.9239),Translation=(X=0,Y=1,Z=0),Scale3D=(X=1,Y=1,Z=1))),(Item=(Type=Bone,Name=\"base_HairStrand_Side_01_l\"),Transform=(Rotation=(X=0,Y=0,Z=0,W=1),Translation=(X= ...(truncated)"
    WeightMinimum=0
    Weight=RigVMVariableNode() {variable=HeadLookingUp}.Value
   )
   &RigUnit_ModifyTransforms(
    ExecutePin=()
    ItemToModify="((Item=(Type=Bone,Name=\"base_Hair_Side_01_r\"),Transform=(Rotation=(X=0.1677,Y=0.2549,Z=-0.0449,W=0.9513),Translation=(X=-1,Y=0,Z=-1),Scale3D=(X=1,Y=1,Z=1))),(Item=(Type=Bone,Name=\"dyn_Hair_Side_02_r\"),Transform=(Rotation=(X=0,Y=-0.342,Z=0,W=0.9397),Transla ...(truncated)"
    Weight=0.8
    WeightMinimum=0
   )
  ]}
 CollapseNode_ContainedGraph {nodes [
   RigVMFunctionEntryNode()
   > GetAlphaFromQuatDelta = RigVMFunctionReferenceNode(
    Rotation Order=ZXY
    child=(Type=Bone,Name="thigh_l")
    debugRotation=false
    outputMinMax=(X=0,Y=1)
    parent=(Type=Null,Name="thigh_l_null")
    premult=1
    rotationMinMax=(X=0,Y=-85)
    triggerAxis=Z
    twistAxis=(X=1,Y=0,Z=0)
   )
   RigVMFunction_AnimEvalRichCurve = &RigVMFunction_AnimEvalRichCurve(
    Curve=(EditorCurveData=(Keys=((InterpMode=RCIM_Cubic,TangentMode=RCTM_User,ArriveTangent=0.4408,LeaveTangent=0.4408),(InterpMode=RCIM_Cubic,TangentMode=RCTM_User,Time=1,Value=1,ArriveTangent=2.3678,LeaveTangent=2.3678))))
    SourceMinimum=0
    TargetMinimum=0
    Value=@GetAlphaFromQuatDelta.alpha
   )
   AlphaInterp = &RigVMFunction_AlphaInterp(InterpSpeedDecreasing=8, InterpSpeedIncreasing=15, Scale=1, bClampResult=False, Value=@RigVMFunction_AnimEvalRichCurve.Result)
   > &RigUnit_ModifyTransforms(
    ItemToModify=((Item=(Type=Bone,Name="base_sword"),Transform=(Rotation=(X=-0.0087,Y=0.0427,Z=0.1992,W=0.979),Translation=(X=-3,Y=-3,Z=0),Scale3D=(X=1,Y=1,Z=1))))
    WeightMinimum=0
    Weight=@AlphaInterp.Result
   )
   > n1 = &RigUnit_SphericalPoseReader(
    ActiveRegionScaleFactors=(PositiveWidth=1,NegativeWidth=1,PositiveHeight=1,NegativeHeight=1)
    ActiveRegionSize=0.025
    Debug=(bDrawDebug=false,bDraw2D=False,bDrawLocalAxes=False,DebugScale=25,DebugSegments=20,DebugThickness=0.25)
    DriverAxis=(X=-1,Y=0,Z=0)
    DriverItem=(Type=Bone,Name="thigh_l")
    FalloffRegionScaleFactors=(PositiveWidth=1,NegativeWidth=1,PositiveHeight=1,NegativeHeight=1)
    FalloffSize=0.2
    FlipHeightScaling=False
    FlipWidthScaling=False
    OptionalParentItem=(Type=Bone,Name="None")
    RotationOffset=(X=122,Y=10,Z=0)
   )
   Multiply = &RigVMFunction_MathDoubleMul(A=@n1.OutputParam, B=RigVMVariableNode() {variable=isCrouchingInPlace}.Value)
   n2 = &RigVMFunction_AlphaInterp(InterpSpeedDecreasing=10, InterpSpeedIncreasing=6, Scale=1, bClampResult=False, Value=@Multiply.Result)
   > &RigUnit_ModifyTransforms(
    ItemToModify=((Item=(Type=Bone,Name="base_sword"),Transform=(Rotation=(X=0.1431,Y=0.1603,Z=-0.1077,W=0.9707),Translation=(X=0,Y=-2,Z=1),Scale3D=(X=1,Y=1,Z=1))))
    WeightMinimum=0
    Weight=&RigVMFunction_MathDoubleMul(B=1, A=@n2.Result).Result
   )
   > RigVMFunctionReturnNode()
  ]}
 New Function_ContainedGraph {nodes [
   RigVMFunctionEntryNode()
   RigVMFunction_MathBoolToInteger = &RigVMFunction_MathBoolToInteger(Value=RigVMVariableNode() {variable=onlyTwist}.Value)
   RigVMFunction_MathQuaternionSwingTwist = &RigVMFunction_MathQuaternionSwingTwist(
    Input=&RigUnit_GetRelativeTransformForItem(
     bChildInitial=False
     bParentInitial=False
     Child=RigVMVariableNode() {variable=child}.Value
     Parent=RigVMVariableNode() {variable=parent}.Value
    ).RelativeTransform.Rotation
    TwistAxis=RigVMVariableNode() {variable=twistAxis}.Value
   )
   n1 = RigVMDispatchNode(
    Values=((),())
    Index=@RigVMFunction_MathBoolToInteger.Result
    Values.0=@RigVMFunction_MathQuaternionSwingTwist.Swing
    Values.1=@RigVMFunction_MathQuaternionSwingTwist.Twist
   ) {template="DISPATCH_RigVMDispatch_SelectInt32(in Index,in Values,out Result)"}
   > RigVMVariableNode(Value=@n1.Result) {variable=entry_outputQuat}
   RigVMFunction_MathQuaternionToEuler = &RigVMFunction_MathQuaternionToEuler(Value=@n1.Result, RotationOrder=RigVMVariableNode() {variable=Rotation Order}.Value)
   > DISPATCH_RigVMDispatch_Print = RigVMDispatchNode(
    ScreenColor=(R=1,G=1,B=1,A=1)
    ScreenDuration=1.5
    Value=@RigVMFunction_MathQuaternionToEuler.Result
    Enabled=RigVMVariableNode() {variable=debugRotation}.Value
   ) {template="DISPATCH_RigVMDispatch_Print(in Prefix,in Value,in Enabled,in ScreenDuration,in ScreenColor)"}
   Cast = RigVMDispatchNode(Value=RigVMVariableNode() {variable=triggerAxis}.Value) {template="DISPATCH_RigVMDispatch_CastEnumToInt(in Value,out Result)"}
   DISPATCH_RigVMDispatch_SelectInt32 = RigVMDispatchNode(
    Values=(0,0,0)
    Index=@Cast.Result
    Values.0=@RigVMFunction_MathQuaternionToEuler.Result.X
    Values.1=@RigVMFunction_MathQuaternionToEuler.Result.Y
    Values.2=@RigVMFunction_MathQuaternionToEuler.Result.Z
   ) {template="DISPATCH_RigVMDispatch_SelectInt32(in Index,in Values,out Result)"}
   n2 = RigVMVariableNode() {variable=rotationMinMax}
   n3 = RigVMVariableNode() {variable=outputMinMax}
   Remap = &RigVMFunction_MathDoubleRemap(
    bClamp=False
    Value=@DISPATCH_RigVMDispatch_SelectInt32.Result
    SourceMinimum=@n2.Value.X
    SourceMaximum=@n2.Value.Y
    TargetMinimum=@n3.Value.X
    TargetMaximum=@n3.Value.Y
   )
   Absolute = &RigVMFunction_MathDoubleAbs(Value=@Remap.Result)
   n4 = RigVMVariableNode() {variable=absoluteValue}
   Multiply = &RigVMFunction_MathDoubleMul(
    A=&RigVMFunction_MathDoubleClamp(Value=RigVMDispatchNode(True=@Absolute.Result, False=@Remap.Result, Condition=@n4.Value) {template="DISPATCH_RigVMDispatch_If(in Condition,in True,in False,out Result)"}.Result).Result
    B=RigVMVariableNode() {variable=premult}.Value
   )
   > RigVMFunctionReturnNode(alpha=@Multiply.Result, rawQuat=RigVMVariableNode() {variable=entry_outputQuat}.Value)
  ]}
 RigVMFunctionLibrary {nodes [GetAlphaFromQuatDelta = RigVMCollapseNode(
    ExecuteContext=()
    Rotation Order=ZYX
    child=(Type=Bone,Name="None")
    outputMinMax=(X=0,Y=1)
    parent=(Type=Bone,Name="None")
    premult=1
    rotationMinMax=(X=0,Y=0)
    twistAxis=(X=1,Y=0,Z=0)
   )]}
 RigVMModel {nodes [
   n1 = &RigUnit_PrepareForExecution()
   RigVMFunction_ControlFlowBranch = &RigVMFunction_ControlFlowBranch(ExecuteContext=@n1.ExecutePin, Condition=RigVMVariableNode() {variable=fkControls}.Value)
   addFKCtrl = RigVMFunctionReferenceNode(
    Custom_FKCtrl=()
    Options_FKCtrl="(pelvis_1_#1=true,spine_3_#2=true,clavicle_5_#3=true,arm_7_#4=true,leg_9_#5=true,head_11_46DC15EF465CD2979 ...(truncated)"
    Shape_Settings="(pelvis_2_#6=(bVisible=True,Name=\"Hexagon_Thin\",Color=(R=0.0117,G=0.8281,B=1,A=1),Transform=(Rotation=(X=0,Y=-0.7071,Z=0,W=0.7071),Translation=(X=2,Y=0,Z=0),Scale3D=(X=5,Y=5,Z=7))),spine_4_#7=(bVi ...(truncated)"
    ExecuteContext=@RigVMFunction_ControlFlowBranch.True
   )
   RigUnit_BeginExecution = &RigUnit_BeginExecution()
   setFKCtrl = RigVMFunctionReferenceNode(ExecuteContext=@RigUnit_BeginExecution.ExecutePin, setFK_Ctrl=RigVMVariableNode() {variable=fkControls}.Value)
   Constraints = RigVMCollapseNode(ExecutePin=@setFKCtrl.ExecuteContext)
   DefiningVariables = RigVMCollapseNode(ExecuteContext=@Constraints.ExecutePin)
   Hair_Setups = RigVMCollapseNode(ExecutePin=@DefiningVariables.ExecuteContext)
   Sword_Setup = RigVMCollapseNode(ExecuteContext=@Hair_Setups.ExecutePin)
   RigVMFunction_MathBoolAnd = &RigVMFunction_MathBoolAnd(A=RigVMVariableNode() {variable=isSkydiving}.Value, B=RigVMVariableNode() {variable=animDynamicsEnabled}.Value)
   RigVMFunction_MathBoolToFloat = &RigVMFunction_MathBoolToFloat(Value=@RigVMFunction_MathBoolAnd.Result)
   AlphaInterp = &RigVMFunction_AlphaInterp(
    InterpSpeedDecreasing=10
    InterpSpeedIncreasing=10
    OutRange=(Min=0,Max=0.75)
    Scale=1
    bClampResult=false
    bMapRange=true
    Value=@RigVMFunction_MathBoolToFloat.Result
   )
   > &RigUnit_ModifyTransforms(
    ItemToModify="((Item=(Type=Bone,Name=\"base_Hair_Back_00\"),Transform=(Rotation=(X=0,Y=0.2588,Z=0,W=0.9659),Translation=(X=0,Y=0,Z=0),Scale3D=(X=1,Y=1,Z=1))),(Item=(Type=Bone,Name=\"base_Hair_Mid_00_l\"),Transform=(Rotation=(X=-0.0151,Y=0.173,Z=0.0858,W=0.9811),Translation= ...(truncated)"
    WeightMinimum=0
    Weight=@AlphaInterp.Result
   )
  ]}
]
hierarchy [
 root {loc=<0 0 0>, rot=<0 0 0 1>}
 pelvis {parent=root, loc=<0 -0.5301 98.0571>, rot=<0.0127 0.707 -0.0127 -0.707>}
 spine_01 {parent=pelvis, loc=<6.1208 -0.9328 0>, rot=<0 0 0.034 -0.9994>}
 spine_02 {parent=spine_01, loc=<6.5997 -0.2095 0>, rot=<0 0 0.0238 -0.9997>}
 spine_03 {parent=spine_02, loc=<6.4914 -0.1237 0>, rot=<0 0 0.0149 -0.9999>}
 spine_04 {parent=spine_03, loc=<6.9598 0.0794 0>, rot=<0 0 -0.0132 -0.9999>}
 spine_05 {parent=spine_04, loc=<6.9987 0 0>, rot=<0 0 -0.0015 -1>}
 neck_01 {parent=spine_05, loc=<11.3138 0.1529 0>, rot=<0 0 -0.2003 -0.9797>}
 neck_02 {parent=neck_01, loc=<4.5627 0 0>, rot=<0 0 0.0504 -0.9987>}
 head {parent=neck_02, loc=<5.3914 0 0>, rot=<0 0 0.115 -0.9934>}
 base_HairStrand_Side_01_l {parent=head, loc=<3.5822 2.3889 -6.6053>, rot=<0.0041 -0.9914 -0.1305 0.0005>}
 dyn_HairStrand_Side_01_l {parent=base_HairStrand_Side_01_l, loc=<0 0 0>, rot=<0 0 0 1>}
 dyn_HairStrand_Side_02_l {parent=dyn_HairStrand_Side_01_l, loc=<3.6663 -0.2677 -0.3088>, rot=<0 0 0 1>}
 dyn_HairStrand_Side_03_l {parent=dyn_HairStrand_Side_02_l, loc=<3.5657 -0.2382 -0.2749>, rot=<0 0 0 1>}
 base_Hair_FrontSide_A_01_r {parent=head, loc=<10.1514 6.7079 3.8503>, rot=<-0.0017 0.4226 0.9063 -0.0037>}
 dyn_Hair_FrontSide_A_01_r {parent=base_Hair_FrontSide_A_01_r, loc=<0 0 0>, rot=<0 0 0 1>}
 dyn_Hair_FrontSide_A_02_r {parent=dyn_Hair_FrontSide_A_01_r, loc=<3.0087 -0.0925 0.8371>, rot=<0 0 0 1>}
 base_Hair_FrontSide_B_01_r {parent=head, loc=<9.0073 5.8581 4.5167>, rot=<-0.0014 0.342 0.9397 -0.0039>}
 dyn_Hair_FrontSide_B_01_r {parent=base_Hair_FrontSide_B_01_r, loc=<0 0 0>, rot=<0 0 0 1>}
 dyn_Hair_FrontSide_B_02_r {parent=dyn_Hair_FrontSide_B_01_r, loc=<2.812 1.3269 0.4821>, rot=<0 0 0 1>}
 base_Hair_FrontSide_B_r_End {parent=dyn_Hair_FrontSide_B_02_r, loc=<2.1066 1.7939 -0.6114>, rot=<0 0 0 1>}
 base_Hair_FrontFleck_Main {parent=head, loc=<8.6425 8.4642 0.7082>, rot=<-0.0029 0.7071 0.7071 -0.0029>}
 dyn_Hair_FrontFleck_Main {parent=base_Hair_FrontFleck_Main, loc=<0 0 0>, rot=<0 0 0 1>}
 dyn_Hair_FrontFleck_C_01 {parent=dyn_Hair_FrontFleck_Main, loc=<2.3651 -2.4557 -0.2059>, rot=<-0.1736 0 0 -0.9848>}
 dyn_Hair_FrontFleck_C_02 {parent=dyn_Hair_FrontFleck_C_01, loc=<1.4114 -0.9266 -0.4757>, rot=<0 0 0 1>}
 dyn_Hair_FrontFleck_B_01 {parent=dyn_Hair_FrontFleck_Main, loc=<2.6724 0.238 0.4009>, rot=<0 0 0 1>}
 dyn_Hair_FrontFleck_B_02 {parent=dyn_Hair_FrontFleck_B_01, loc=<1.774 -1.0638 -0.3659>, rot=<0 0 0 1>}
 dyn_Hair_FrontFleck_A {parent=dyn_Hair_FrontFleck_Main, loc=<2.7332 2.3958 -0.5823>, rot=<0.1736 0 0 -0.9848>}
 base_Hair_FrontSide_A_01_l {parent=head, loc=<7.4993 7.1557 -3.855>, rot=<0.0037 -0.9063 -0.4226 0.0017>}
 dyn_Hair_FrontSide_A_01_l {parent=base_Hair_FrontSide_A_01_l, loc=<0 0 0>, rot=<0 0 0 1>}
 dyn_Hair_FrontSide_A_02_l {parent=dyn_Hair_FrontSide_A_01_l, loc=<2.3764 -1.0756 -0.3735>, rot=<0 0 0 1>}
 base_Hair_FrontSide_B_01_l {parent=head, loc=<6.1865 5.1014 -5.268>, rot=<0.0039 -0.9397 -0.342 0.0014>}
 dyn_Hair_FrontSide_B_01_l {parent=base_Hair_FrontSide_B_01_l, loc=<0 0 0>, rot=<0 0 0 1>}
 base_HairStrand_Side_A_01_r {parent=head, loc=<1.7435 -1.7236 7.6762>, rot=<0 0 1 -0.0041>}
 dyn_HairStrand_Side_A_01_r {parent=base_HairStrand_Side_A_01_r, loc=<0 0 0>, rot=<0 0 0 1>}
 dyn_HairStrand_Side_A_02_r {parent=dyn_HairStrand_Side_A_01_r, loc=<4.549 -0.7471 0.5874>, rot=<0 0 0 1>}
 HairStrand_Side_A_r_End {parent=dyn_HairStrand_Side_A_02_r, loc=<3.543 -0.2679 -0.8274>, rot=<0 0 0 1>}
 base_HairStrand_Side_B_01_r {parent=head, loc=<0.903 -2.5173 5.6574>, rot=<0 0 1 -0.0041>}
 dyn_HairStrand_Side_B_01_r {parent=base_HairStrand_Side_B_01_r, loc=<0 0 0>, rot=<0 0 0 1>}
 dyn_HairStrand_Side_B_02_r {parent=dyn_HairStrand_Side_B_01_r, loc=<4.0813 0.1021 0.3008>, rot=<0 0 0 1>}
 base_Hair_Side_00_l {parent=head, loc=<7.9434 -2.5438 -8.3596>, rot=<0.0041 -1 0 0>}
 dyn_Hair_Side_00_l {parent=base_Hair_Side_00_l, loc=<0 0 0>, rot=<0 0 0 1>}
 base_Hair_Side_01_l {parent=dyn_Hair_Side_00_l, loc=<2.9761 -0.2608 -0.4889>, rot=<0 0 0 -1>}
 dyn_Hair_Side_01_l {parent=base_Hair_Side_01_l, loc=<0 0 0>, rot=<0 0 0 1>}
 dyn_Hair_Side_02_l {parent=dyn_Hair_Side_01_l, loc=<4.979 -0.2324 0.7247>, rot=<0 0 0 1>}
 base_Hair_Mid_00_l {parent=head, loc=<9.0424 -8.3704 -6.0544>, rot=<0.0037 -0.9063 0.4226 -0.0017>}
 dyn_Hair_Mid_00_l {parent=base_Hair_Mid_00_l, loc=<0 0 0>, rot=<0 0 0 1>}
 base_Hair_Mid_01_l {parent=dyn_Hair_Mid_00_l, loc=<3.1622 -0.7074 0.4343>, rot=<0 0 0 -1>}
 dyn_Hair_Mid_01_l {parent=base_Hair_Mid_01_l, loc=<0 0 0>, rot=<0 0 0 1>}
 dyn_Hair_Mid_02_l {parent=dyn_Hair_Mid_01_l, loc=<4.7713 -0.579 1.821>, rot=<0 0 0 1>}
 base_Hair_Back_00 {parent=head, loc=<9.6935 -10.543 0>, rot=<0.0029 -0.7071 0.7071 -0.0029>}
 dyn_Hair_Back_00 {parent=base_Hair_Back_00, loc=<0 0 0>, rot=<0 0 0 1>}
 base_Hair_Back_01 {parent=dyn_Hair_Back_00, loc=<3.7953 0 0.4797>, rot=<0 0 0 -1>}
 dyn_Hair_Back_01 {parent=base_Hair_Back_01, loc=<0 0 0>, rot=<0 0 0 1>}
 dyn_Hair_Back_02 {parent=dyn_Hair_Back_01, loc=<4.8222 0 1.7785>, rot=<0 0 0 1>}
 base_Hair_Mid_00_r {parent=head, loc=<9.0485 -8.373 6.0484>, rot=<0.0014 -0.342 0.9397 -0.0039>}
 dyn_Hair_Mid_00_r {parent=base_Hair_Mid_00_r, loc=<0 0 0>, rot=<0 0 0 1>}
 base_Hair_Mid_01_r {parent=dyn_Hair_Mid_00_r, loc=<3.1682 0.7663 0.3077>, rot=<0 0 0 -1>}
 dyn_Hair_Mid_01_r {parent=base_Hair_Mid_01_r, loc=<0 0 0>, rot=<0 0 0 1>}
 dyn_Hair_Mid_02_r {parent=dyn_Hair_Mid_01_r, loc=<4.771 0.8864 1.6928>, rot=<0 0 0 1>}
 base_Hair_Side_00_r {parent=head, loc=<7.9434 -2.5438 8.3596>, rot=<0 0 1 -0.0041>}
 dyn_Hair_Side_00_r {parent=base_Hair_Side_00_r, loc=<0 0 0>, rot=<0 0 0 1>}
 base_Hair_Side_01_r {parent=dyn_Hair_Side_00_r, loc=<2.9465 0.9095 -0.4904>, rot=<0 0 0 -1>}
 dyn_Hair_Side_01_r {parent=base_Hair_Side_01_r, loc=<0 0 0>, rot=<0 0 0 1>}
 dyn_Hair_Side_02_r {parent=dyn_Hair_Side_01_r, loc=<4.9744 0.3263 1.2511>, rot=<0 0 0 1>}
 clavicle_l {parent=spine_05, loc=<5.4831 -0.1053 -3.2946>, rot=<0.0465 -0.7289 -0.009 -0.6829>}
 upperarm_l {parent=clavicle_l, loc=<9.8658 -1.2136 -0.4105>, rot=<0.0164 -0.3245 -0.0269 -0.9454>}
 base_shoulderPad_pivot_l {parent=upperarm_l, loc=<-0.7419 0 -0.2877>, rot=<0 0 0 1>}
 dyn_shoulderPad_pivot_l {parent=base_shoulderPad_pivot_l, loc=<0 0 0>, rot=<0 0 0 1>}
 dyn_shoulderPad_top_l {parent=dyn_shoulderPad_pivot_l, loc=<0.9749 0.7306 4.8178>, rot=<0 0 0 1>}
 dyn_shoulderPad_Ring_l {parent=dyn_shoulderPad_top_l, loc=<-4.6159 -0.7306 -5.7564>, rot=<0 0.1126 0 -0.9936>}
 clavicle_r {parent=spine_05, loc=<5.4831 -0.1053 3.2946>, rot=<0.7289 0.0465 -0.6829 0.009>}
 upperarm_r {parent=clavicle_r, loc=<-9.8658 1.2136 0.4105>, rot=<0.0164 -0.3245 -0.0269 -0.9454>}
 base_shoulderPad_pivot_r {parent=upperarm_r, loc=<0.1007 0 -0.1594>, rot=<0 0 0 1>}
 dyn_shoulderPad_pivot_r {parent=base_shoulderPad_pivot_r, loc=<0 0 0>, rot=<0 0 0 1>}
 dyn_shoulderPad_top_r {parent=dyn_shoulderPad_pivot_r, loc=<0.732 -0.7251 -4.3942>, rot=<0 0.0422 0 -0.9991>}
 neck_01_null {parent=spine_05, type=Null, loc=<11.3138 0.1529 0>, rot=<0 0 -0.2003 -0.9797>}
 base_beltPivot_Start {parent=spine_01, loc=<4.7072 0.5663 10.5162>, rot=<0.0335 -0.8406 0.0041 -0.5406>}
 dyn_beltPivotStart {parent=base_beltPivot_Start, loc=<0 0 0>, rot=<0 0 0 1>}
 dyn_beltPivotMid {parent=dyn_beltPivotStart, loc=<10.0727 0 0>, rot=<0 0 0 1>}
 dyn_beltPivotEnd {parent=dyn_beltPivotMid, loc=<16.4345 0 0>, rot=<0 0 0 1>}
 base_sword {parent=dyn_beltPivotEnd, loc=<6.9751 0.0512 -3.8634>, rot=<0.2918 -0.4434 0.4318 -0.7293>}
 dyn_sword {parent=base_sword, loc=<0 0 0>, rot=<0 0 0 1>}
 thigh_l {parent=pelvis, loc=<-1.0855 -0.2109 -8.7053>, rot=<0.0009 0.0355 0.0155 -0.9993>}
 calf_l {parent=thigh_l, loc=<-42.5284 -0.0019 -0.7297>, rot=<0.0006 -0.0161 -0.0391 -0.9991>}
 thigh_twist_01_l {parent=thigh_l, loc=<-27.952 0 0>, rot=<0.0013 0 0 -1>}
 base_SkirtFlap_00_l {parent=pelvis, loc=<3.6879 -10.0909 -8.2744>, rot=<0.0057 -0.7071 0.7067 -0.0232>}
 dyn_SkirtFlap_00_l {parent=base_SkirtFlap_00_l, loc=<0 0 0>, rot=<0 0 0 1>}
 base_SkirtFlap_01_l {parent=dyn_SkirtFlap_00_l, loc=<7.6115 0.0004 -0.0001>, rot=<0 0 0 1>}
 dyn_SkirtFlap_01_l {parent=base_SkirtFlap_01_l, loc=<0 0 0>, rot=<0 0 0 1>}
 dyn_SkirtFlap_02_l {parent=dyn_SkirtFlap_01_l, loc=<14.0845 0 0>, rot=<0 0 0 1>}
 dyn_SkirtFlap_03_l {parent=dyn_SkirtFlap_02_l, loc=<15.9816 0 0>, rot=<0 0 0 1>}
 dyn_SkirtFlap_04_l {parent=dyn_SkirtFlap_03_l, loc=<12.6231 0 0>, rot=<0 -0.0205 0 -0.9998>}
 thigh_l_null {parent=pelvis, type=Null, loc=<-1.0855 -0.2109 -8.7053>, rot=<0.0009 0.0355 0.0155 -0.9993>}
]
variables [
 HeadLookingUp {type=float, default=0}
 HeadLookingUp_Crouching {type=double, default=0}
 HeadLookingUp_Extreme {type=double, default=0}
 HeadTiltingSideways {type=double, default=0}
 animDynamicsEnabled {type=bool, default=True}
 fkControls {type=bool, default=False}
 isCrouchingInPlace {type=double, default=0}
 isSkydiving {type=bool, default=True}
]
