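# Animation Blueprint for the QuirkLuster female face-accessory mesh, adding secondary motion to hair and a sword bone.
# It blends poses copied from the attach parent and Body_Mesh by limb branches, then caches them for hair-strand and belt-pivot bone adjustments.
# A long Anim Dynamics chain runs on preDeformCache, using head-driven spherical limits on front hair.
# The output layers DefaultSlot over the cached dynamics from dyn_sword down.
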AnimBlueprint /BRCosmetics/Characters/Player/Female/Medium/Bodies/F_MED_QuirkLuster/Meshes/Parts/F_MED_QuirkLuster_FaceAcc_AnimBP
target_skeleton=&/F_MED_QuirkLuster_FaceAcc_Skeleton
graphs [
 AnimGraph {nodes [
   Output Pose(Result=Layered blend per bone(BasePose=Use cached pose 'DynamicsCache'(), BlendPoses_0=Slot 'DefaultSlot'()) {layerSetup [{branchFilters [{boneName=dyn_sword}]}]}) {layerGroup=None}
   n1 = Layered blend per bone(
    BasePose=Copy Pose From Mesh() {bUseAttachedParent=true}
    BlendPoses_0=Copy Pose From Mesh(SourceMeshComponent=Get Body_Mesh()) {bUseAttachedParent=true}
   ) {layerSetup [{branchFilters [
       {boneName=clavicle_l}
       {boneName=clavicle_r}
       {boneName=upperarm_r}
       {boneName=upperarm_l}
       {boneName=thigh_l}
       {boneName=calf_l}
       {boneName=thigh_twist_01_l}
      ]}]}
   Save cached pose 'CRCache_Input'(Pose=@n1) {cacheName=CRCache_Input}
   n2 = Anim Dynamics - Bone: dyn_Hair_Side_00_l!(ComponentPose=Local To Component(LocalPose=Use cached pose 'preDeformCache'()), Alpha=0, GravityOverride:Vector="0,0,-980") -> (Pose) {
    angularDampingOverride=0.9999
    angularSpringConstant=5000
    bAngularSpring=true
    bOverrideAngularDamping=true
    bUseGravityOverride=true
    gravityScale=0.35
    simulationSpace=RootRelative
    boundBone {boneName=dyn_Hair_Side_00_l}
    componentAppliedLinearAccClamp=<7500 7500 7500>
    componentLinearAccScale=<1 1 1>
    componentLinearVelScale=<1 1 1>
    externalForce {z=-980}
    gravityOverride {z=-980}
    physicsBodyDefinitions [{
      boundBone {boneName=dyn_Hair_Side_00_l}
      boxExtents=<20 30 30>
      constraintSetup {angularLimitsMax=<0 0 5>, angularLimitsMin=<0 -25 -15>, angularTarget=<1 0 0>}
      localJointOffset=<1 0 0>
     }]
    simSpaceSettings {maxAngularAcceleration=35, maxAngularVelocity=7, simSpaceAngularAlpha=1}
   }
   n3 = Anim Dynamics - Bone: dyn_Hair_Side_01_l!(ComponentPose=@n2, Alpha=0.5, GravityOverride:Vector="0,0,-980") -> (Pose) {
    angularDampingOverride=0.9999
    angularSpringConstant=5000
    bAngularSpring=true
    bOverrideAngularDamping=true
    bUseGravityOverride=true
    gravityScale=0.35
    simulationSpace=RootRelative
    alphaScaleBiasClamp {bInterpResult=true, interpSpeedDecreasing=7, interpSpeedIncreasing=7}
    boundBone {boneName=dyn_Hair_Side_01_l}
    componentAppliedLinearAccClamp=<7500 7500 7500>
    componentLinearAccScale=<1 1 1>
    componentLinearVelScale=<1 1 1>
    externalForce {z=-980}
    gravityOverride {z=-980}
    physicsBodyDefinitions [{
      boundBone {boneName=dyn_Hair_Side_01_l}
      boxExtents=<20 30 30>
      constraintSetup {angularLimitsMax=<0 0 5>, angularLimitsMin=<0 -25 -15>, angularTarget=<1 0 0>}
      localJointOffset=<1 0 0>
     }]
    simSpaceSettings {maxAngularAcceleration=35, maxAngularVelocity=7, simSpaceAngularAlpha=1}
   }
   n4 = Anim Dynamics - Bone: dyn_Hair_Side_02_l!(ComponentPose=@n3, Alpha=0.5, GravityOverride:Vector="0,0,-980") -> (Pose) {
    angularDampingOverride=0.9999
    angularSpringConstant=5000
    bAngularSpring=true
    bOverrideAngularDamping=true
    bUseGravityOverride=true
    gravityScale=0.35
    simulationSpace=RootRelative
    alphaScaleBiasClamp {bInterpResult=true, interpSpeedDecreasing=7, interpSpeedIncreasing=7}
    boundBone {boneName=dyn_Hair_Side_02_l}
    componentAppliedLinearAccClamp=<7500 7500 7500>
    componentLinearAccScale=<1 1 1>
    componentLinearVelScale=<1 1 1>
    externalForce {z=-980}
    gravityOverride {z=-980}
    physicsBodyDefinitions [{
      boundBone {boneName=dyn_Hair_Side_02_l}
      boxExtents=<20 30 30>
      constraintSetup {angularLimitsMax=<0 0 5>, angularLimitsMin=<0 -25 -15>, angularTarget=<1 0 0>}
      localJointOffset=<1 0 0>
     }]
    simSpaceSettings {maxAngularAcceleration=35, maxAngularVelocity=7, simSpaceAngularAlpha=1}
   }
   n5 = Anim Dynamics - Bone: dyn_Hair_Mid_00_l!(ComponentPose=@n4, Alpha=0.5, GravityOverride:Vector="0,0,-980") -> (Pose) {
    angularDampingOverride=0.9999
    angularSpringConstant=5000
    bAngularSpring=true
    bOverrideAngularDamping=true
    bUseGravityOverride=true
    gravityScale=0.35
    simulationSpace=RootRelative
    alphaScaleBiasClamp {bInterpResult=true, interpSpeedDecreasing=7, interpSpeedIncreasing=7}
    boundBone {boneName=dyn_Hair_Mid_00_l}
    componentAppliedLinearAccClamp=<7500 7500 7500>
    componentLinearAccScale=<1 1 1>
    componentLinearVelScale=<1 1 1>
    externalForce {z=-980}
    gravityOverride {z=-980}
    physicsBodyDefinitions [{
      boundBone {boneName=dyn_Hair_Mid_00_l}
      boxExtents=<20 30 30>
      constraintSetup {angularLimitsMax=<0 0 10>, angularLimitsMin=<0 -20 -10>, angularTarget=<1 0 0>}
      localJointOffset=<2 0 0>
     }]
    simSpaceSettings {maxAngularAcceleration=35, maxAngularVelocity=7, simSpaceAngularAlpha=1}
   }
   n6 = Anim Dynamics - Bone: dyn_Hair_Mid_01_l!(ComponentPose=@n5, Alpha=0.5, GravityOverride:Vector="0,0,-980") -> (Pose) {
    angularDampingOverride=0.9999
    angularSpringConstant=5000
    bAngularSpring=true
    bOverrideAngularDamping=true
    bUseGravityOverride=true
    gravityScale=0.35
    simulationSpace=RootRelative
    alphaScaleBiasClamp {bInterpResult=true, interpSpeedDecreasing=7, interpSpeedIncreasing=7}
    boundBone {boneName=dyn_Hair_Mid_01_l}
    componentAppliedLinearAccClamp=<7500 7500 7500>
    componentLinearAccScale=<1 1 1>
    componentLinearVelScale=<1 1 1>
    externalForce {z=-980}
    gravityOverride {z=-980}
    physicsBodyDefinitions [{
      boundBone {boneName=dyn_Hair_Mid_01_l}
      boxExtents=<20 30 30>
      constraintSetup {angularLimitsMax=<0 0 10>, angularLimitsMin=<0 -20 -10>, angularTarget=<1 0 0>}
      localJointOffset=<2 0 0>
     }]
    simSpaceSettings {maxAngularAcceleration=35, maxAngularVelocity=7, simSpaceAngularAlpha=1}
   }
   n7 = Anim Dynamics - Bone: dyn_Hair_Mid_02_l!(ComponentPose=@n6, Alpha=0.5, GravityOverride:Vector="0,0,-980") -> (Pose) {
    angularDampingOverride=0.9999
    angularSpringConstant=5000
    bAngularSpring=true
    bOverrideAngularDamping=true
    bUseGravityOverride=true
    gravityScale=0.35
    simulationSpace=RootRelative
    alphaScaleBiasClamp {bInterpResult=true, interpSpeedDecreasing=7, interpSpeedIncreasing=7}
    boundBone {boneName=dyn_Hair_Mid_02_l}
    componentAppliedLinearAccClamp=<7500 7500 7500>
    componentLinearAccScale=<1 1 1>
    componentLinearVelScale=<1 1 1>
    externalForce {z=-980}
    gravityOverride {z=-980}
    physicsBodyDefinitions [{
      boundBone {boneName=dyn_Hair_Mid_02_l}
      boxExtents=<20 30 30>
      constraintSetup {angularLimitsMax=<0 0 10>, angularLimitsMin=<0 -20 -10>, angularTarget=<1 0 0>}
      localJointOffset=<2 0 0>
     }]
    simSpaceSettings {maxAngularAcceleration=35, maxAngularVelocity=7, simSpaceAngularAlpha=1}
   }
   n8 = Anim Dynamics - Bone: dyn_Hair_Back_00!(ComponentPose=@n7, Alpha=0.5, GravityOverride:Vector="0,0,-980") -> (Pose) {
    angularDampingOverride=0.9999
    angularSpringConstant=5000
    bAngularSpring=true
    bOverrideAngularDamping=true
    bUseGravityOverride=true
    gravityScale=0.35
    simulationSpace=RootRelative
    alphaScaleBiasClamp {bInterpResult=true, interpSpeedDecreasing=7, interpSpeedIncreasing=7}
    boundBone {boneName=dyn_Hair_Back_00}
    componentAppliedLinearAccClamp=<7500 7500 7500>
    componentLinearAccScale=<1 1 1>
    componentLinearVelScale=<1 1 1>
    externalForce {z=-980}
    gravityOverride {z=-980}
    physicsBodyDefinitions [{
      boundBone {boneName=dyn_Hair_Back_00}
      boxExtents=<20 30 30>
      constraintSetup {angularLimitsMax=<0 0 10>, angularLimitsMin=<0 -15 -10>, angularTarget=<1 0 0>}
      localJointOffset=<2 0 0>
     }]
    simSpaceSettings {maxAngularAcceleration=35, maxAngularVelocity=7, simSpaceAngularAlpha=1}
   }
   n9 = Anim Dynamics - Bone: dyn_Hair_Back_01!(ComponentPose=@n8, Alpha=0.5, GravityOverride:Vector="0,0,-980") -> (Pose) {
    angularDampingOverride=0.9999
    angularSpringConstant=5000
    bAngularSpring=true
    bOverrideAngularDamping=true
    bUseGravityOverride=true
    gravityScale=0.35
    simulationSpace=RootRelative
    alphaScaleBiasClamp {bInterpResult=true, interpSpeedDecreasing=7, interpSpeedIncreasing=7}
    boundBone {boneName=dyn_Hair_Back_01}
    componentAppliedLinearAccClamp=<7500 7500 7500>
    componentLinearAccScale=<1 1 1>
    componentLinearVelScale=<1 1 1>
    externalForce {z=-980}
    gravityOverride {z=-980}
    physicsBodyDefinitions [{
      boundBone {boneName=dyn_Hair_Back_01}
      boxExtents=<20 30 30>
      constraintSetup {angularLimitsMax=<0 0 10>, angularLimitsMin=<0 -15 -10>, angularTarget=<1 0 0>}
      localJointOffset=<2 0 0>
     }]
    simSpaceSettings {maxAngularAcceleration=35, maxAngularVelocity=7, simSpaceAngularAlpha=1}
   }
   n10 = Anim Dynamics - Bone: dyn_Hair_Back_02!(ComponentPose=@n9, Alpha=0.5, GravityOverride:Vector="0,0,-980") -> (Pose) {
    angularDampingOverride=0.9999
    angularSpringConstant=5000
    bAngularSpring=true
    bOverrideAngularDamping=true
    bUseGravityOverride=true
    gravityScale=0.35
    simulationSpace=RootRelative
    alphaScaleBiasClamp {bInterpResult=true, interpSpeedDecreasing=7, interpSpeedIncreasing=7}
    boundBone {boneName=dyn_Hair_Back_02}
    componentAppliedLinearAccClamp=<7500 7500 7500>
    componentLinearAccScale=<1 1 1>
    componentLinearVelScale=<1 1 1>
    externalForce {z=-980}
    gravityOverride {z=-980}
    physicsBodyDefinitions [{
      boundBone {boneName=dyn_Hair_Back_02}
      boxExtents=<20 30 30>
      constraintSetup {angularLimitsMax=<0 0 10>, angularLimitsMin=<0 -15 -10>, angularTarget=<1 0 0>}
      localJointOffset=<2 0 0>
     }]
    simSpaceSettings {maxAngularAcceleration=35, maxAngularVelocity=7, simSpaceAngularAlpha=1}
   }
   n11 = Anim Dynamics - Bone: dyn_Hair_Mid_00_r!(ComponentPose=@n10, Alpha=0.5, GravityOverride:Vector="0,0,-980") -> (Pose) {
    angularDampingOverride=0.9999
    angularSpringConstant=5000
    bAngularSpring=true
    bOverrideAngularDamping=true
    bUseGravityOverride=true
    gravityScale=0.35
    simulationSpace=RootRelative
    alphaScaleBiasClamp {bInterpResult=true, interpSpeedDecreasing=7, interpSpeedIncreasing=7}
    boundBone {boneName=dyn_Hair_Mid_00_r}
    componentAppliedLinearAccClamp=<7500 7500 7500>
    componentLinearAccScale=<1 1 1>
    componentLinearVelScale=<1 1 1>
    externalForce {z=-980}
    gravityOverride {z=-980}
    physicsBodyDefinitions [{
      boundBone {boneName=dyn_Hair_Mid_00_r}
      boxExtents=<20 30 30>
      constraintSetup {angularLimitsMax=<0 0 10>, angularLimitsMin=<0 -20 -10>, angularTarget=<1 0 0>}
      localJointOffset=<2 0 0>
     }]
    simSpaceSettings {maxAngularAcceleration=35, maxAngularVelocity=7, simSpaceAngularAlpha=1}
   }
   n12 = Anim Dynamics - Bone: dyn_Hair_Mid_01_r!(ComponentPose=@n11, Alpha=0.5, GravityOverride:Vector="0,0,-980") -> (Pose) {
    angularDampingOverride=0.9999
    angularSpringConstant=5000
    bAngularSpring=true
    bOverrideAngularDamping=true
    bUseGravityOverride=true
    gravityScale=0.35
    simulationSpace=RootRelative
    alphaScaleBiasClamp {bInterpResult=true, interpSpeedDecreasing=7, interpSpeedIncreasing=7}
    boundBone {boneName=dyn_Hair_Mid_01_r}
    componentAppliedLinearAccClamp=<7500 7500 7500>
    componentLinearAccScale=<1 1 1>
    componentLinearVelScale=<1 1 1>
    externalForce {z=-980}
    gravityOverride {z=-980}
    physicsBodyDefinitions [{
      boundBone {boneName=dyn_Hair_Mid_01_r}
      boxExtents=<20 30 30>
      constraintSetup {angularLimitsMax=<0 0 10>, angularLimitsMin=<0 -20 -10>, angularTarget=<1 0 0>}
      localJointOffset=<2 0 0>
     }]
    simSpaceSettings {maxAngularAcceleration=35, maxAngularVelocity=7, simSpaceAngularAlpha=1}
   }
   n13 = Anim Dynamics - Bone: dyn_Hair_Mid_02_r!(ComponentPose=@n12, Alpha=0.5, GravityOverride:Vector="0,0,-980") -> (Pose) {
    angularDampingOverride=0.9999
    angularSpringConstant=5000
    bAngularSpring=true
    bOverrideAngularDamping=true
    bUseGravityOverride=true
    gravityScale=0.35
    simulationSpace=RootRelative
    alphaScaleBiasClamp {bInterpResult=true, interpSpeedDecreasing=7, interpSpeedIncreasing=7}
    boundBone {boneName=dyn_Hair_Mid_02_r}
    componentAppliedLinearAccClamp=<7500 7500 7500>
    componentLinearAccScale=<1 1 1>
    componentLinearVelScale=<1 1 1>
    externalForce {z=-980}
    gravityOverride {z=-980}
    physicsBodyDefinitions [{
      boundBone {boneName=dyn_Hair_Mid_02_r}
      boxExtents=<20 30 30>
      constraintSetup {angularLimitsMax=<0 0 10>, angularLimitsMin=<0 -20 -10>, angularTarget=<1 0 0>}
      localJointOffset=<2 0 0>
     }]
    simSpaceSettings {maxAngularAcceleration=35, maxAngularVelocity=7, simSpaceAngularAlpha=1}
   }
   n14 = Anim Dynamics - Bone: dyn_Hair_Side_00_r!(ComponentPose=@n13, Alpha=0.5, GravityOverride:Vector="0,0,-980") -> (Pose) {
    angularDampingOverride=0.9999
    angularSpringConstant=5000
    bAngularSpring=true
    bOverrideAngularDamping=true
    bUseGravityOverride=true
    gravityScale=0.35
    simulationSpace=RootRelative
    alphaScaleBiasClamp {bInterpResult=true, interpSpeedDecreasing=7, interpSpeedIncreasing=7}
    boundBone {boneName=dyn_Hair_Side_00_r}
    componentAppliedLinearAccClamp=<7500 7500 7500>
    componentLinearAccScale=<1 1 1>
    componentLinearVelScale=<1 1 1>
    externalForce {z=-980}
    gravityOverride {z=-980}
    physicsBodyDefinitions [{
      boundBone {boneName=dyn_Hair_Side_00_r}
      boxExtents=<20 30 30>
      constraintSetup {angularLimitsMax=<0 0 10>, angularLimitsMin=<0 -25 -10>, angularTarget=<1 0 0>}
      localJointOffset=<1 0 0>
     }]
    simSpaceSettings {maxAngularAcceleration=35, maxAngularVelocity=7, simSpaceAngularAlpha=1}
   }
   n15 = Anim Dynamics - Bone: dyn_Hair_Side_01_r!(ComponentPose=@n14, Alpha=0.5, GravityOverride:Vector="0,0,-980") -> (Pose) {
    angularDampingOverride=0.9999
    angularSpringConstant=5000
    bAngularSpring=true
    bOverrideAngularDamping=true
    bUseGravityOverride=true
    gravityScale=0.35
    simulationSpace=RootRelative
    alphaScaleBiasClamp {bInterpResult=true, interpSpeedDecreasing=7, interpSpeedIncreasing=7}
    boundBone {boneName=dyn_Hair_Side_01_r}
    componentAppliedLinearAccClamp=<7500 7500 7500>
    componentLinearAccScale=<1 1 1>
    componentLinearVelScale=<1 1 1>
    externalForce {z=-980}
    gravityOverride {z=-980}
    physicsBodyDefinitions [{
      boundBone {boneName=dyn_Hair_Side_01_r}
      boxExtents=<20 30 30>
      constraintSetup {angularLimitsMax=<0 0 10>, angularLimitsMin=<0 -25 -10>, angularTarget=<1 0 0>}
      localJointOffset=<1 0 0>
     }]
    simSpaceSettings {maxAngularAcceleration=35, maxAngularVelocity=7, simSpaceAngularAlpha=1}
   }
   n16 = Anim Dynamics - Bone: dyn_Hair_Side_02_r!(ComponentPose=@n15, Alpha=0.5, GravityOverride:Vector="0,0,-980") -> (Pose) {
    angularDampingOverride=0.9999
    angularSpringConstant=5000
    bAngularSpring=true
    bOverrideAngularDamping=true
    bUseGravityOverride=true
    gravityScale=0.35
    simulationSpace=RootRelative
    alphaScaleBiasClamp {bInterpResult=true, interpSpeedDecreasing=7, interpSpeedIncreasing=7}
    boundBone {boneName=dyn_Hair_Side_02_r}
    componentAppliedLinearAccClamp=<7500 7500 7500>
    componentLinearAccScale=<1 1 1>
    componentLinearVelScale=<1 1 1>
    externalForce {z=-980}
    gravityOverride {z=-980}
    physicsBodyDefinitions [{
      boundBone {boneName=dyn_Hair_Side_02_r}
      boxExtents=<20 30 30>
      constraintSetup {angularLimitsMax=<0 0 10>, angularLimitsMin=<0 -25 -10>, angularTarget=<1 0 0>}
      localJointOffset=<1 0 0>
     }]
    simSpaceSettings {maxAngularAcceleration=35, maxAngularVelocity=7, simSpaceAngularAlpha=1}
   }
   n17 = Anim Dynamics - Bone: dyn_HairStrand_Side_01_l!(ComponentPose=@n16, Alpha=1, GravityOverride:Vector="0,0,-980") -> (Pose) {
    angularDampingOverride=0.9999
    angularSpringConstant=5000
    bAngularSpring=true
    bOverrideAngularDamping=true
    bUseGravityOverride=true
    gravityScale=0.35
    simulationSpace=RootRelative
    boundBone {boneName=dyn_HairStrand_Side_01_l}
    componentAppliedLinearAccClamp=<7500 7500 7500>
    componentLinearAccScale=<1 1 1>
    componentLinearVelScale=<1 1 1>
    externalForce {z=-980}
    gravityOverride {z=-980}
    physicsBodyDefinitions [{
      boundBone {boneName=dyn_HairStrand_Side_01_l}
      boxExtents=<20 30 30>
      constraintSetup {angularLimitsMax=<0 7 45>, angularLimitsMin=<0 -25 -15>, angularTarget=<1 0 0>}
      localJointOffset=<2 0 0>
     }]
    simSpaceSettings {maxAngularAcceleration=48, maxAngularVelocity=8, simSpaceAngularAlpha=1}
   }
   n18 = Anim Dynamics - Bone: dyn_HairStrand_Side_02_l!(ComponentPose=@n17, Alpha=1, GravityOverride:Vector="0,0,-980") -> (Pose) {
    angularDampingOverride=0.9999
    angularSpringConstant=5000
    bAngularSpring=true
    bOverrideAngularDamping=true
    bUseGravityOverride=true
    gravityScale=0.35
    simulationSpace=RootRelative
    boundBone {boneName=dyn_HairStrand_Side_02_l}
    componentAppliedLinearAccClamp=<7500 7500 7500>
    componentLinearAccScale=<1 1 1>
    componentLinearVelScale=<1 1 1>
    externalForce {z=-980}
    gravityOverride {z=-980}
    physicsBodyDefinitions [{
      boundBone {boneName=dyn_HairStrand_Side_02_l}
      boxExtents=<20 30 30>
      constraintSetup {angularLimitsMax=<0 7 25>, angularLimitsMin=<0 -25 -25>, angularTarget=<1 0 0>}
      localJointOffset=<2 0 0>
     }]
    simSpaceSettings {maxAngularAcceleration=48, maxAngularVelocity=8, simSpaceAngularAlpha=1}
   }
   n19 = Anim Dynamics - Bone: dyn_HairStrand_Side_03_l!(ComponentPose=@n18, Alpha=1, GravityOverride:Vector="0,0,-980") -> (Pose) {
    angularDampingOverride=0.9999
    angularSpringConstant=5000
    bAngularSpring=true
    bOverrideAngularDamping=true
    bUseGravityOverride=true
    gravityScale=0.35
    simulationSpace=RootRelative
    boundBone {boneName=dyn_HairStrand_Side_03_l}
    componentAppliedLinearAccClamp=<7500 7500 7500>
    componentLinearAccScale=<1 1 1>
    componentLinearVelScale=<1 1 1>
    externalForce {z=-980}
    gravityOverride {z=-980}
    physicsBodyDefinitions [{
      boundBone {boneName=dyn_HairStrand_Side_03_l}
      boxExtents=<20 30 30>
      constraintSetup {angularLimitsMax=<0 7 25>, angularLimitsMin=<0 -25 -25>, angularTarget=<1 0 0>}
      localJointOffset=<2 0 0>
     }]
    simSpaceSettings {maxAngularAcceleration=48, maxAngularVelocity=8, simSpaceAngularAlpha=1}
   }
   n20 = Anim Dynamics - Bone: dyn_HairStrand_Side_B_01_r!(ComponentPose=@n19, Alpha=1, GravityOverride:Vector="0,0,-980") -> (Pose) {
    angularDampingOverride=0.9999
    angularSpringConstant=5000
    bAngularSpring=true
    bOverrideAngularDamping=true
    bUseGravityOverride=true
    gravityScale=0.35
    simulationSpace=RootRelative
    boundBone {boneName=dyn_HairStrand_Side_B_01_r}
    componentAppliedLinearAccClamp=<7500 7500 7500>
    componentLinearAccScale=<1 1 1>
    componentLinearVelScale=<1 1 1>
    externalForce {z=-980}
    gravityOverride {z=-980}
    physicsBodyDefinitions [{
      boundBone {boneName=dyn_HairStrand_Side_B_01_r}
      boxExtents=<20 30 30>
      constraintSetup {angularLimitsMax=<0 0 10>, angularLimitsMin=<0 -25 -10>, angularTarget=<1 0 0>}
      localJointOffset=<2 0 0>
     }]
    simSpaceSettings {maxAngularAcceleration=30, maxAngularVelocity=6, simSpaceAngularAlpha=1}
   }
   n21 = Anim Dynamics - Bone: dyn_HairStrand_Side_B_02_r!(ComponentPose=@n20, Alpha=1, GravityOverride:Vector="0,0,-980") -> (Pose) {
    angularDampingOverride=0.9999
    angularSpringConstant=5000
    bAngularSpring=true
    bOverrideAngularDamping=true
    bUseGravityOverride=true
    gravityScale=0.35
    simulationSpace=RootRelative
    boundBone {boneName=dyn_HairStrand_Side_B_02_r}
    componentAppliedLinearAccClamp=<7500 7500 7500>
    componentLinearAccScale=<1 1 1>
    componentLinearVelScale=<1 1 1>
    externalForce {z=-980}
    gravityOverride {z=-980}
    physicsBodyDefinitions [{
      boundBone {boneName=dyn_HairStrand_Side_B_02_r}
      boxExtents=<20 30 30>
      constraintSetup {angularLimitsMax=<0 0 10>, angularLimitsMin=<0 -25 -10>, angularTarget=<1 0 0>}
      localJointOffset=<2 0 0>
     }]
    simSpaceSettings {maxAngularAcceleration=30, maxAngularVelocity=6, simSpaceAngularAlpha=1}
   }
   n22 = Anim Dynamics - Bone: dyn_HairStrand_Side_A_01_r!(ComponentPose=@n21, Alpha=1, GravityOverride:Vector="0,0,-980") -> (Pose) {
    angularDampingOverride=0.9999
    angularSpringConstant=5000
    bAngularSpring=true
    bOverrideAngularDamping=true
    bUseGravityOverride=true
    gravityScale=0.35
    simulationSpace=RootRelative
    boundBone {boneName=dyn_HairStrand_Side_A_01_r}
    componentAppliedLinearAccClamp=<7500 7500 7500>
    componentLinearAccScale=<1 1 1>
    componentLinearVelScale=<1 1 1>
    externalForce {z=-980}
    gravityOverride {z=-980}
    physicsBodyDefinitions [{
      boundBone {boneName=dyn_HairStrand_Side_A_01_r}
      boxExtents=<20 30 30>
      constraintSetup {angularLimitsMax=<0 0 10>, angularLimitsMin=<0 -25 -10>, angularTarget=<1 0 0>}
      localJointOffset=<2 0 0>
     }]
    simSpaceSettings {maxAngularAcceleration=30, maxAngularVelocity=6, simSpaceAngularAlpha=1}
   }
   n23 = Anim Dynamics - Bone: dyn_HairStrand_Side_A_02_r!(ComponentPose=@n22, Alpha=1, GravityOverride:Vector="0,0,-980") -> (Pose) {
    angularDampingOverride=0.9999
    angularSpringConstant=5000
    bAngularSpring=true
    bOverrideAngularDamping=true
    bUseGravityOverride=true
    gravityScale=0.35
    simulationSpace=RootRelative
    boundBone {boneName=dyn_HairStrand_Side_A_02_r}
    componentAppliedLinearAccClamp=<7500 7500 7500>
    componentLinearAccScale=<1 1 1>
    componentLinearVelScale=<1 1 1>
    externalForce {z=-980}
    gravityOverride {z=-980}
    physicsBodyDefinitions [{
      boundBone {boneName=dyn_HairStrand_Side_A_02_r}
      boxExtents=<20 30 30>
      constraintSetup {angularLimitsMax=<0 0 10>, angularLimitsMin=<0 -25 -10>, angularTarget=<1 0 0>}
      localJointOffset=<2 0 0>
     }]
    simSpaceSettings {maxAngularAcceleration=30, maxAngularVelocity=6, simSpaceAngularAlpha=1}
   }
   n24 = Anim Dynamics - Bone: dyn_Hair_FrontSide_B_01_r!(ComponentPose=@n23, Alpha=1, GravityOverride:Vector="0,0,-980") -> (Pose) {
    angularDampingOverride=0.9999
    angularSpringConstant=3000
    bAngularSpring=true
    bOverrideAngularDamping=true
    bUseGravityOverride=true
    gravityScale=0.35
    simulationSpace=RootRelative
    boundBone {boneName=dyn_Hair_FrontSide_B_01_r}
    componentAppliedLinearAccClamp=<7500 7500 7500>
    componentLinearAccScale=<1 1 1>
    componentLinearVelScale=<1 1 1>
    externalForce {z=-980}
    gravityOverride {z=-980}
    physicsBodyDefinitions [{
      boundBone {boneName=dyn_Hair_FrontSide_B_01_r}
      boxExtents=<15 20 20>
      constraintSetup {angularLimitsMax=<0 0 3>, angularLimitsMin=<0 -2 -20>, angularTarget=<1 0 0>}
      localJointOffset=<2 0 0>
     }]
    relativeSpaceBone {boneName=spine_03}
    simSpaceSettings {maxAngularAcceleration=12, maxAngularVelocity=5, simSpaceAngularAlpha=1}
   }
   n25 = Anim Dynamics - Bone: dyn_Hair_FrontSide_B_02_r!(ComponentPose=@n24, Alpha=1, GravityOverride:Vector="0,0,-980") -> (Pose) {
    angularDampingOverride=0.9999
    angularSpringConstant=3000
    bAngularSpring=true
    bOverrideAngularDamping=true
    bUseGravityOverride=true
    gravityScale=0.35
    simulationSpace=RootRelative
    boundBone {boneName=dyn_Hair_FrontSide_B_02_r}
    componentAppliedLinearAccClamp=<7500 7500 7500>
    componentLinearAccScale=<1 1 1>
    componentLinearVelScale=<1 1 1>
    externalForce {z=-980}
    gravityOverride {z=-980}
    physicsBodyDefinitions [{
      boundBone {boneName=dyn_Hair_FrontSide_B_02_r}
      boxExtents=<15 20 20>
      constraintSetup {angularLimitsMax=<0 0 3>, angularLimitsMin=<0 -5 -20>, angularTarget=<1 0 0>}
      localJointOffset=<2 0 0>
     }]
    relativeSpaceBone {boneName=spine_03}
    simSpaceSettings {maxAngularAcceleration=12, maxAngularVelocity=5, simSpaceAngularAlpha=1}
   }
   n26 = Anim Dynamics - Bone: base_Hair_FrontSide_B_r_End!(ComponentPose=@n25, Alpha=1, GravityOverride:Vector="0,0,-980") -> (Pose) {
    angularDampingOverride=0.9999
    angularSpringConstant=3000
    bAngularSpring=true
    bOverrideAngularDamping=true
    bUseGravityOverride=true
    gravityScale=0.35
    simulationSpace=RootRelative
    boundBone {boneName=base_Hair_FrontSide_B_r_End}
    componentAppliedLinearAccClamp=<7500 7500 7500>
    componentLinearAccScale=<1 1 1>
    componentLinearVelScale=<1 1 1>
    externalForce {z=-980}
    gravityOverride {z=-980}
    physicsBodyDefinitions [{
      boundBone {boneName=base_Hair_FrontSide_B_r_End}
      boxExtents=<15 20 20>
      constraintSetup {angularLimitsMax=<0 0 3>, angularLimitsMin=<0 -10 -20>, angularTarget=<1 0 0>}
      localJointOffset=<2 0 0>
     }]
    relativeSpaceBone {boneName=spine_03}
    simSpaceSettings {maxAngularAcceleration=12, maxAngularVelocity=5, simSpaceAngularAlpha=1}
   }
   n27 = Anim Dynamics - Bone: dyn_Hair_FrontSide_A_01_r!(ComponentPose=@n26, Alpha=1, GravityOverride:Vector="0,0,-980") -> (Pose) {
    angularDampingOverride=0.9999
    angularSpringConstant=3000
    bAngularSpring=true
    bOverrideAngularDamping=true
    bUseGravityOverride=true
    gravityScale=0.35
    simulationSpace=RootRelative
    boundBone {boneName=dyn_Hair_FrontSide_A_01_r}
    componentAppliedLinearAccClamp=<7500 7500 7500>
    componentLinearAccScale=<1 1 1>
    componentLinearVelScale=<1 1 1>
    externalForce {z=-980}
    gravityOverride {z=-980}
    physicsBodyDefinitions [{
      boundBone {boneName=dyn_Hair_FrontSide_A_01_r}
      boxExtents=<15 20 20>
      constraintSetup {angularLimitsMax=<0 0 20>, angularLimitsMin=<0 -2 -20>, angularTarget=<1 0 0>}
      localJointOffset=<1 0 0>
     }]
    relativeSpaceBone {boneName=spine_03}
    simSpaceSettings {maxAngularAcceleration=30, maxAngularVelocity=6, simSpaceAngularAlpha=1}
   }
   n28 = Anim Dynamics - Bone: dyn_Hair_FrontSide_A_02_r!(ComponentPose=@n27, Alpha=1, GravityOverride:Vector="0,0,-980") -> (Pose) {
    angularDampingOverride=0.9999
    angularSpringConstant=3000
    bAngularSpring=true
    bOverrideAngularDamping=true
    bUseGravityOverride=true
    gravityScale=0.35
    simulationSpace=RootRelative
    boundBone {boneName=dyn_Hair_FrontSide_A_02_r}
    componentAppliedLinearAccClamp=<7500 7500 7500>
    componentLinearAccScale=<1 1 1>
    componentLinearVelScale=<1 1 1>
    externalForce {z=-980}
    gravityOverride {z=-980}
    physicsBodyDefinitions [{
      boundBone {boneName=dyn_Hair_FrontSide_A_02_r}
      boxExtents=<15 20 20>
      constraintSetup {angularLimitsMax=<0 0 20>, angularLimitsMin=<0 -5 -20>, angularTarget=<1 0 0>}
      localJointOffset=<1 0 0>
     }]
    relativeSpaceBone {boneName=spine_03}
    simSpaceSettings {maxAngularAcceleration=30, maxAngularVelocity=6, simSpaceAngularAlpha=1}
   }
   n29 = Anim Dynamics - Bone: dyn_Hair_FrontFleck_Main!(ComponentPose=@n28, Alpha=1, GravityOverride:Vector="0,0,-980") -> (Pose) {
    angularDampingOverride=0.9999
    angularSpringConstant=3000
    bOverrideAngularDamping=true
    bUseGravityOverride=true
    bUseSphericalLimits=true
    gravityScale=0.35
    simulationSpace=RootRelative
    boundBone {boneName=dyn_Hair_FrontFleck_Main}
    componentAppliedLinearAccClamp=<7500 7500 7500>
    componentLinearAccScale=<1 1 1>
    componentLinearVelScale=<1 1 1>
    externalForce {z=-980}
    gravityOverride {z=-980}
    physicsBodyDefinitions [{
      boundBone {boneName=dyn_Hair_FrontFleck_Main}
      boxExtents=<15 20 20>
      constraintSetup {angularLimitsMax=<0 -2 5>, angularLimitsMin=<0 -4 -5>, angularTarget=<1 0 0>}
      localJointOffset=<1.5 0 0>
     }]
    relativeSpaceBone {boneName=spine_03}
    simSpaceSettings {maxAngularAcceleration=30, maxAngularVelocity=6, simSpaceAngularAlpha=1}
    sphericalLimits [{limitRadius=10, drivingBone {boneName=head}, sphereLocalOffset=<5 -1.5 0>}]
   }
   n30 = Anim Dynamics - Bone: dyn_Hair_FrontFleck_A!(ComponentPose=@n29, Alpha=1, GravityOverride:Vector="0,0,-980") -> (Pose) {
    angularDampingOverride=0.9999
    angularSpringConstant=3000
    bOverrideAngularDamping=true
    bUseGravityOverride=true
    bUseSphericalLimits=true
    gravityScale=0.35
    simulationSpace=RootRelative
    boundBone {boneName=dyn_Hair_FrontFleck_A}
    componentAppliedLinearAccClamp=<7500 7500 7500>
    componentLinearAccScale=<1 1 1>
    componentLinearVelScale=<1 1 1>
    externalForce {z=-980}
    gravityOverride {z=-980}
    physicsBodyDefinitions [{
      boundBone {boneName=dyn_Hair_FrontFleck_A}
      boxExtents=<15 20 20>
      constraintSetup {angularLimitsMax=<0 0 5>, angularLimitsMin=<0 -5 -15>, angularTarget=<1 0 0>}
      localJointOffset=<1.5 0 0>
     }]
    relativeSpaceBone {boneName=spine_03}
    simSpaceSettings {maxAngularAcceleration=35, maxAngularVelocity=7, simSpaceAngularAlpha=1}
    sphericalLimits [{limitRadius=10, drivingBone {boneName=head}, sphereLocalOffset=<4.5 -1.5 0>}]
   }
   n31 = Anim Dynamics - Bone: dyn_Hair_FrontFleck_B_01!(ComponentPose=@n30, Alpha=1, GravityOverride:Vector="0,0,-980") -> (Pose) {
    angularDampingOverride=0.9999
    angularSpringConstant=3000
    bOverrideAngularDamping=true
    bUseGravityOverride=true
    bUseSphericalLimits=true
    gravityScale=0.35
    simulationSpace=RootRelative
    boundBone {boneName=dyn_Hair_FrontFleck_B_01}
    componentAppliedLinearAccClamp=<7500 7500 7500>
    componentLinearAccScale=<1 1 1>
    componentLinearVelScale=<1 1 1>
    externalForce {z=-980}
    gravityOverride {z=-980}
    physicsBodyDefinitions [{
      boundBone {boneName=dyn_Hair_FrontFleck_B_01}
      boxExtents=<15 20 20>
      constraintSetup {angularLimitsMax=<0 0 10>, angularLimitsMin=<0 -5 -10>, angularTarget=<1 0 0>}
      localJointOffset=<1.5 0 0>
     }]
    relativeSpaceBone {boneName=spine_03}
    simSpaceSettings {maxAngularAcceleration=30, maxAngularVelocity=6, simSpaceAngularAlpha=1}
    sphericalLimits [{limitRadius=10, drivingBone {boneName=head}, sphereLocalOffset=<5 -1.5 0>}]
   }
   n32 = Anim Dynamics - Bone: dyn_Hair_FrontFleck_B_02!(ComponentPose=@n31, Alpha=1, GravityOverride:Vector="0,0,-980") -> (Pose) {
    angularDampingOverride=0.9999
    angularSpringConstant=3000
    bOverrideAngularDamping=true
    bUseGravityOverride=true
    bUseSphericalLimits=true
    gravityScale=0.35
    simulationSpace=RootRelative
    boundBone {boneName=dyn_Hair_FrontFleck_B_02}
    componentAppliedLinearAccClamp=<7500 7500 7500>
    componentLinearAccScale=<1 1 1>
    componentLinearVelScale=<1 1 1>
    externalForce {z=-980}
    gravityOverride {z=-980}
    physicsBodyDefinitions [{
      boundBone {boneName=dyn_Hair_FrontFleck_B_02}
      boxExtents=<15 20 20>
      constraintSetup {angularLimitsMax=<0 0 3>, angularLimitsMin=<0 -10 -20>, angularTarget=<1 0 0>}
      localJointOffset=<1.5 0 0>
     }]
    relativeSpaceBone {boneName=spine_03}
    simSpaceSettings {maxAngularAcceleration=35, maxAngularVelocity=7, simSpaceAngularAlpha=1}
    sphericalLimits [{limitRadius=10, drivingBone {boneName=head}, sphereLocalOffset=<5 -1.5 0>}]
   }
   n33 = Anim Dynamics - Bone: dyn_Hair_FrontFleck_C_01!(ComponentPose=@n32, Alpha=1, GravityOverride:Vector="0,0,-980") -> (Pose) {
    angularDampingOverride=0.9999
    angularSpringConstant=3000
    bOverrideAngularDamping=true
    bUseGravityOverride=true
    bUseSphericalLimits=true
    gravityScale=0.35
    simulationSpace=RootRelative
    boundBone {boneName=dyn_Hair_FrontFleck_C_01}
    componentAppliedLinearAccClamp=<7500 7500 7500>
    componentLinearAccScale=<1 1 1>
    componentLinearVelScale=<1 1 1>
    externalForce {z=-980}
    gravityOverride {z=-980}
    physicsBodyDefinitions [{
      boundBone {boneName=dyn_Hair_FrontFleck_C_01}
      boxExtents=<15 20 20>
      constraintSetup {angularLimitsMax=<0 -2 10>, angularLimitsMin=<0 -12 -5>, angularTarget=<1 0 0>}
      localJointOffset=<1.5 0 0>
     }]
    relativeSpaceBone {boneName=spine_03}
    simSpaceSettings {maxAngularAcceleration=30, maxAngularVelocity=6, simSpaceAngularAlpha=1}
    sphericalLimits [{limitRadius=10, drivingBone {boneName=head}, sphereLocalOffset=<5 -1.75 0.25>}]
   }
   n34 = Anim Dynamics - Bone: dyn_Hair_FrontFleck_C_02!(ComponentPose=@n33, Alpha=1, GravityOverride:Vector="0,0,-980") -> (Pose) {
    angularDampingOverride=0.9999
    angularSpringConstant=3000
    bOverrideAngularDamping=true
    bUseGravityOverride=true
    bUseSphericalLimits=true
    gravityScale=0.35
    simulationSpace=RootRelative
    boundBone {boneName=dyn_Hair_FrontFleck_C_02}
    componentAppliedLinearAccClamp=<7500 7500 7500>
    componentLinearAccScale=<1 1 1>
    componentLinearVelScale=<1 1 1>
    externalForce {z=-980}
    gravityOverride {z=-980}
    physicsBodyDefinitions [{
      boundBone {boneName=dyn_Hair_FrontFleck_C_02}
      boxExtents=<15 20 20>
      constraintSetup {angularLimitsMax=<0 -2 10>, angularLimitsMin=<0 -12 -5>, angularTarget=<1 0 0>}
      localJointOffset=<1.5 0 0>
     }]
    relativeSpaceBone {boneName=spine_03}
    simSpaceSettings {maxAngularAcceleration=30, maxAngularVelocity=6, simSpaceAngularAlpha=1}
    sphericalLimits [{limitRadius=10, drivingBone {boneName=head}, sphereLocalOffset=<5 -1.75 0.25>}]
   }
   n35 = Anim Dynamics - Bone: dyn_Hair_FrontSide_A_01_l!(ComponentPose=@n34, Alpha=1, GravityOverride:Vector="0,0,-980") -> (Pose) {
    angularDampingOverride=0.9999
    angularSpringConstant=3000
    bAngularSpring=true
    bOverrideAngularDamping=true
    bUseGravityOverride=true
    bUseSphericalLimits=true
    gravityScale=0.35
    simulationSpace=RootRelative
    boundBone {boneName=dyn_Hair_FrontSide_A_01_l}
    componentAppliedLinearAccClamp=<7500 7500 7500>
    componentLinearAccScale=<1 1 1>
    componentLinearVelScale=<1 1 1>
    externalForce {z=-980}
    gravityOverride {z=-980}
    physicsBodyDefinitions [{
      boundBone {boneName=dyn_Hair_FrontSide_A_01_l}
      boxExtents=<15 20 20>
      constraintSetup {angularLimitsMax=<0 -1 20>, angularLimitsMin=<0 -6 -20>, angularTarget=<1 0 0>}
      localJointOffset=<1.5 0 0>
     }]
    relativeSpaceBone {boneName=spine_03}
    simSpaceSettings {maxAngularAcceleration=30, maxAngularVelocity=6, simSpaceAngularAlpha=1}
    sphericalLimits [{limitRadius=10, drivingBone {boneName=head}, sphereLocalOffset=<5.5 -1.25 1.75>}]
   }
   n36 = Anim Dynamics - Bone: dyn_Hair_FrontSide_A_02_l!(ComponentPose=@n35, Alpha=1, GravityOverride:Vector="0,0,-980") -> (Pose) {
    angularDampingOverride=0.9999
    angularSpringConstant=3000
    bAngularSpring=true
    bOverrideAngularDamping=true
    bUseGravityOverride=true
    bUseSphericalLimits=true
    gravityScale=0.35
    simulationSpace=RootRelative
    boundBone {boneName=dyn_Hair_FrontSide_A_02_l}
    componentAppliedLinearAccClamp=<7500 7500 7500>
    componentLinearAccScale=<1 1 1>
    componentLinearVelScale=<1 1 1>
    externalForce {z=-980}
    gravityOverride {z=-980}
    physicsBodyDefinitions [{
      boundBone {boneName=dyn_Hair_FrontSide_A_02_l}
      boxExtents=<15 20 20>
      constraintSetup {angularLimitsMax=<25 0 20>, angularLimitsMin=<0 -10 -20>, angularTarget=<1 0 0>}
      localJointOffset=<1.5 0 0>
     }]
    relativeSpaceBone {boneName=spine_03}
    simSpaceSettings {maxAngularAcceleration=30, maxAngularVelocity=6, simSpaceAngularAlpha=1}
    sphericalLimits [{limitRadius=10, drivingBone {boneName=head}, sphereLocalOffset=<5 -1.25 2>}]
   }
   n37 = Anim Dynamics - Bone: dyn_Hair_FrontSide_B_01_l!(ComponentPose=@n36, Alpha=1, GravityOverride:Vector="0,0,-980") -> (Pose) {
    angularDampingOverride=0.9999
    angularSpringConstant=3000
    bAngularSpring=true
    bOverrideAngularDamping=true
    bUseGravityOverride=true
    bUseSphericalLimits=true
    gravityScale=0.35
    simulationSpace=RootRelative
    boundBone {boneName=dyn_Hair_FrontSide_B_01_l}
    componentAppliedLinearAccClamp=<7500 7500 7500>
    componentLinearAccScale=<1 1 1>
    componentLinearVelScale=<1 1 1>
    externalForce {z=-980}
    gravityOverride {z=-980}
    physicsBodyDefinitions [{
      boundBone {boneName=dyn_Hair_FrontSide_B_01_l}
      boxExtents=<15 20 20>
      constraintSetup {angularLimitsMax=<0 0 20>, angularLimitsMin=<0 -10 -20>, angularTarget=<1 0 0>}
      localJointOffset=<1.5 0 0>
     }]
    relativeSpaceBone {boneName=spine_03}
    simSpaceSettings {maxAngularAcceleration=30, maxAngularVelocity=6, simSpaceAngularAlpha=1}
    sphericalLimits [{limitRadius=10, drivingBone {boneName=head}, sphereLocalOffset=<5 -1.5 2.5>}]
   }
   n38 = Anim Dynamics - Bone: dyn_sword!(ComponentPose=@n37, Alpha=1, GravityOverride:Vector="0,0,-980") -> (Pose) {
    angularDampingOverride=0.9999
    angularSpringConstant=3000
    bAngularSpring=true
    bOverrideAngularDamping=true
    bUseGravityOverride=true
    gravityScale=0.35
    simulationSpace=RootRelative
    boundBone {boneName=dyn_sword}
    componentAppliedLinearAccClamp=<7500 7500 7500>
    componentLinearAccScale=<1 1 1>
    componentLinearVelScale=<1 1 1>
    externalForce {z=-980}
    gravityOverride {z=-980}
    physicsBodyDefinitions [{
      boundBone {boneName=dyn_sword}
      boxExtents=<15 20 20>
      constraintSetup {angularLimitsMax=<0 0 20>, angularLimitsMin=<0 0 -20>, angularTarget=<1 0 0>}
      localJointOffset=<0.25 0 0>
     }]
    simSpaceSettings {maxAngularAcceleration=30, maxAngularVelocity=6, simSpaceAngularAlpha=1}
   }
   Save cached pose 'DynamicsCache'(Pose=Component To Local(ComponentPose=@n38)) {cacheName=DynamicsCache}
   n39 = "Transform (Modify) Bone - Bone: base_HairStrand_Side_B_01_r"!(
    ComponentPose=Local To Component(LocalPose=Use cached pose 'CRCache_Input'())
    bAlphaBoolEnabled:bool=True
    Translation="0,0,-0.1"
    Rotation="0,0,0"
   ) -> (Pose) {
    alphaInputType=Bool
    rotationMode=BMM_Additive
    rotationSpace=BCS_ParentBoneSpace
    scaleSpace=BCS_BoneSpace
    translationMode=BMM_Additive
    translationSpace=BCS_ParentBoneSpace
    alphaBoolBlend {blendOption=Linear}
    boneToModify {boneName=base_HairStrand_Side_B_01_r}
    translation {z=-0.1}
   }
   n40 = "Transform (Modify) Bone - Bone: dyn_HairStrand_Side_B_02_r"!(ComponentPose=@n39, bAlphaBoolEnabled:bool=True, Translation="0,0,0", Rotation="12,5,0") -> (Pose) {
    alphaInputType=Bool
    rotationMode=BMM_Additive
    rotationSpace=BCS_ParentBoneSpace
    scaleSpace=BCS_BoneSpace
    translationMode=BMM_Additive
    translationSpace=BCS_ParentBoneSpace
    alphaBoolBlend {blendOption=Linear}
    boneToModify {boneName=dyn_HairStrand_Side_B_02_r}
    rotation {pitch=12, yaw=5}
   }
   n41 = "Transform (Modify) Bone - Bone: base_beltPivot_Start"!(ComponentPose=@n40, bAlphaBoolEnabled:bool=True, Translation="0,0,0", Rotation="0,0,0") -> (Pose) {
    alphaInputType=Bool
    rotationMode=BMM_Replace
    rotationSpace=BCS_ParentBoneSpace
    scaleSpace=BCS_BoneSpace
    translationMode=BMM_Replace
    translationSpace=BCS_ParentBoneSpace
    alphaBoolBlend {blendOption=Linear}
    boneToModify {boneName=base_beltPivot_Start}
   }
   n42 = "Transform (Modify) Bone - Bone: dyn_beltPivotStart"!(ComponentPose=@n41, bAlphaBoolEnabled:bool=True, Translation="0,0,0", Rotation="0,0,0") -> (Pose) {
    alphaInputType=Bool
    rotationMode=BMM_Replace
    rotationSpace=BCS_ParentBoneSpace
    scaleSpace=BCS_BoneSpace
    translationMode=BMM_Replace
    translationSpace=BCS_ParentBoneSpace
    alphaBoolBlend {blendOption=Linear}
    boneToModify {boneName=dyn_beltPivotStart}
   }
   n43 = "Transform (Modify) Bone - Bone: dyn_beltPivotMid"!(ComponentPose=@n42, bAlphaBoolEnabled:bool=True, Translation="0,0,0", Rotation="0,0,0") -> (Pose) {
    alphaInputType=Bool
    rotationMode=BMM_Replace
    rotationSpace=BCS_ParentBoneSpace
    scaleSpace=BCS_BoneSpace
    translationMode=BMM_Replace
    translationSpace=BCS_ParentBoneSpace
    alphaBoolBlend {blendOption=Linear}
    boneToModify {boneName=dyn_beltPivotMid}
   }
   n44 = "Transform (Modify) Bone - Bone: dyn_beltPivotEnd"!(ComponentPose=@n43, bAlphaBoolEnabled:bool=True, Translation="0,0,0", Rotation="0,0,0") -> (Pose) {
    alphaInputType=Bool
    rotationMode=BMM_Replace
    rotationSpace=BCS_ParentBoneSpace
    scaleSpace=BCS_BoneSpace
    translationMode=BMM_Replace
    translationSpace=BCS_ParentBoneSpace
    alphaBoolBlend {blendOption=Linear}
    boneToModify {boneName=dyn_beltPivotEnd}
   }
   n45 = "Transform (Modify) Bone - Bone: base_SkirtFlap_00_l"!(ComponentPose=@n44, bAlphaBoolEnabled:bool=True, Translation="0,0,0", Rotation="0,0,0") -> (Pose) {
    alphaInputType=Bool
    rotationMode=BMM_Replace
    rotationSpace=BCS_ParentBoneSpace
    scaleSpace=BCS_BoneSpace
    translationMode=BMM_Replace
    translationSpace=BCS_ParentBoneSpace
    alphaBoolBlend {blendOption=Linear}
    boneToModify {boneName=base_SkirtFlap_00_l}
   }
   n46 = "Transform (Modify) Bone - Bone: dyn_SkirtFlap_00_l"!(ComponentPose=@n45, bAlphaBoolEnabled:bool=True, Translation="0,0,0", Rotation="0,0,0") -> (Pose) {
    alphaInputType=Bool
    rotationMode=BMM_Replace
    rotationSpace=BCS_ParentBoneSpace
    scaleSpace=BCS_BoneSpace
    translationMode=BMM_Replace
    translationSpace=BCS_ParentBoneSpace
    alphaBoolBlend {blendOption=Linear}
    boneToModify {boneName=dyn_SkirtFlap_00_l}
   }
   n47 = "Transform (Modify) Bone - Bone: base_SkirtFlap_01_l"!(ComponentPose=@n46, bAlphaBoolEnabled:bool=True, Translation="0,0,0", Rotation="0,0,0") -> (Pose) {
    alphaInputType=Bool
    rotationMode=BMM_Replace
    rotationSpace=BCS_ParentBoneSpace
    scaleSpace=BCS_BoneSpace
    translationMode=BMM_Replace
    translationSpace=BCS_ParentBoneSpace
    alphaBoolBlend {blendOption=Linear}
    boneToModify {boneName=base_SkirtFlap_01_l}
   }
   n48 = Control Rig!(Source=Component To Local(ComponentPose=@n47), Alpha, isSkydiving:bool=True, animDynamicsEnabled:bool=True) -> (Pose) {controlRigAssetReference {blueprintRigClass=&/F_MED_QuirkLuster_FaceAcc_CtrlRig.F_MED_QuirkLuster_FaceAcc_CtrlRig_C}}
   Save cached pose 'preDeformCache'(Pose=@n48) {cacheName=preDeformCache}
  ]}
 EventGraph {nodes [
   Event BlueprintUpdateAnimation()
   > n1 = Sequence!(execute) -> (then_0, then_1, then_2:Exec)
   n2 = Branch(execute=@n1.then_1, Condition=Get bIsSkyDiving())
   > Set backHairAlpha(backHairAlpha:double=0.5) -> (Output_Get:double)
   Set backHairAlpha(execute=@n2.else, backHairAlpha=1)
   Branch(execute=@n1.then_2, Condition=Get bIsInFrontEnd())
   > Set CustomIdleIsPlayingFloat(CustomIdleIsPlayingFloat:float=GetCurveValue(self=GetOwnerFrontEndAnimInstance(), CurveName=Frontend_Idle_Bypass)) -> (Output_Get:float)
   n3 = "Gravity Override State Function (Parameter Asset) (Nativized)"(
    execute=@n1.then_0
    CharacterPartAnimInstance=Self-Reference()
    GameTimeInSeconds=GetGameTimeInSeconds()
    Parameters=&/F_MED_QuirkLuster_FaceAcc_Parameters
   )
   > Set Gravity Override(Gravity Override=@n3)
   n4 = Get Body_Mesh()
   > Is Valid(InputObject=@n4)
   n5 = GetSocketBoneName!(self:Skinned Mesh Component Object Reference=@n4, InSocketName:Name=upperarm_r) -> (ReturnValue:Name=None)
   > n6 = Set isBodyAttached(isBodyAttached:bool="Not Equal Exactly (String)"(A="To String (Name)"(InName=@n5), B=None)) -> (Output_Get:bool)
   > Branch(Condition=@n6)
   n7 = BreakTransform(InTransform=GetSocketTransform(self=Get Body_Mesh(), InSocketName=base_beltPivot_Start, TransformSpace=RTS_ParentBoneSpace))
   > Set base_beltPivot_Start_Translation(base_beltPivot_Start_Translation:Vector=@n7) -> (Output_Get:Vector)
   > Set base_beltPivot_Start_Rotation(base_beltPivot_Start_Rotation:Rotator=@n7.Rotation) -> (Output_Get:Rotator)
   n8 = BreakTransform(InTransform=GetSocketTransform(self=Get Body_Mesh(), InSocketName=dyn_beltPivotStart, TransformSpace=RTS_ParentBoneSpace))
   > Set dyn_beltPivotStart_Translation(dyn_beltPivotStart_Translation:Vector=@n8) -> (Output_Get:Vector)
   > Set dyn_beltPivotStart_Rotation(dyn_beltPivotStart_Rotation:Rotator=@n8.Rotation) -> (Output_Get:Rotator)
   n9 = BreakTransform(InTransform=GetSocketTransform(self=Get Body_Mesh(), InSocketName=dyn_beltPivotMid, TransformSpace=RTS_ParentBoneSpace))
   > Set dyn_beltPivotMid_Translation(dyn_beltPivotMid_Translation:Vector=@n9) -> (Output_Get:Vector)
   > Set dyn_beltPivotMid_Rotation(dyn_beltPivotMid_Rotation:Rotator=@n9.Rotation) -> (Output_Get:Rotator)
   n10 = BreakTransform(InTransform=GetSocketTransform(self=Get Body_Mesh(), InSocketName=dyn_beltPivotEnd, TransformSpace=RTS_ParentBoneSpace))
   > Set dyn_beltPivotEnd_Translation(dyn_beltPivotEnd_Translation:Vector=@n10) -> (Output_Get:Vector)
   > Set dyn_beltPivotEnd_Rotation(dyn_beltPivotEnd_Rotation:Rotator=@n10.Rotation) -> (Output_Get:Rotator)
   n11 = BreakTransform(InTransform=GetSocketTransform(self=Get Body_Mesh(), InSocketName=base_SkirtFlap_00_l, TransformSpace=RTS_ParentBoneSpace))
   > Set base_SkirtFlap_00_l_Translation(base_SkirtFlap_00_l_Translation:Vector=@n11) -> (Output_Get:Vector)
   > Set base_SkirtFlap_00_l_Rotation(base_SkirtFlap_00_l_Rotation:Rotator=@n11.Rotation) -> (Output_Get:Rotator)
   n12 = BreakTransform(InTransform=GetSocketTransform(self=Get Body_Mesh(), InSocketName=dyn_SkirtFlap_00_l, TransformSpace=RTS_ParentBoneSpace))
   > Set dyn_SkirtFlap_00_l_Translation(dyn_SkirtFlap_00_l_Translation:Vector=@n12) -> (Output_Get:Vector)
   > Set dyn_SkirtFlap_00_l_Rotation(dyn_SkirtFlap_00_l_Rotation:Rotator=@n12.Rotation) -> (Output_Get:Rotator)
   n13 = BreakTransform(InTransform=GetSocketTransform(self=Get Body_Mesh(), InSocketName=base_SkirtFlap_01_l, TransformSpace=RTS_ParentBoneSpace))
   > Set base_SkirtFlap_01_l_Translation(base_SkirtFlap_01_l_Translation:Vector=@n13) -> (Output_Get:Vector)
   > Set base_SkirtFlap_01_l_Rotation(base_SkirtFlap_01_l_Rotation:Rotator=@n13.Rotation) -> (Output_Get:Rotator)
   n14 = Event OnCharacterCustomizationLoaded()
   > Set Body_Mesh(Body_Mesh=Get Array Item with EFort Custom Part Type(Array=@n14.CustomMeshParts))
  ]}
]
variables [
 backHairAlpha {type=number}
 base_beltPivot_Start_Rotation {type=Rotator}
 base_beltPivot_Start_Translation {type=Vector}
 base_SkirtFlap_00_l_Rotation {type=Rotator}
 base_SkirtFlap_00_l_Translation {type=Vector}
 base_SkirtFlap_01_l_Rotation {type=Rotator}
 base_SkirtFlap_01_l_Translation {type=Vector}
 Body_Mesh {type=SkeletalMeshComponent}
 CustomIdleIsPlayingFloat {type=number}
 dyn_beltPivotEnd_Rotation {type=Rotator}
 dyn_beltPivotEnd_Translation {type=Vector}
 dyn_beltPivotMid_Rotation {type=Rotator}
 dyn_beltPivotMid_Translation {type=Vector}
 dyn_beltPivotStart_Rotation {type=Rotator}
 dyn_beltPivotStart_Translation {type=Vector}
 dyn_SkirtFlap_00_l_Rotation {type=Rotator}
 dyn_SkirtFlap_00_l_Translation {type=Vector}
 Gravity Override {type=Vector}
 isBodyAttached {type=boolean}
]
